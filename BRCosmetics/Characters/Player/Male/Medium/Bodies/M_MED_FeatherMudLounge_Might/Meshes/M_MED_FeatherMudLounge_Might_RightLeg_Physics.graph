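# Physics asset for the right leg of the FeatherMudLounge Might male medium body cosmetic, giving its shorts secondary-motion bodies around the thigh.
# Four small base capsules each attach to the single thigh_r capsule through constraints with free swing and twist.
# Each base anchors a larger dyn_shorts capsule with narrow swing limits and a zero twist limit, forming four short, tightly bounded cloth-like flaps.
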PhysicsAsset /BRCosmetics/Characters/Player/Male/Medium/Bodies/M_MED_FeatherMudLounge_Might/Meshes/M_MED_FeatherMudLounge_Might_RightLeg_Physics
bodies {
 base_shorts_fmd_r_01 [{shapeType=Capsule, length=1, radius=1, shapeName=dyn_shorts_r_01_capsule, rotation=<rot -90 0 0>}]
 base_shorts_fmd_r_02 [{shapeType=Capsule, length=1, radius=1, shapeName=dyn_shorts_r_01_capsule, rotation=<rot -90 0 0>}]
 base_shorts_fmd_r_03 [{shapeType=Capsule, length=1, radius=1, shapeName=dyn_shorts_r_01_capsule, rotation=<rot -90 0 0>}]
 base_shorts_fmd_r_04 [{shapeType=Capsule, length=1, radius=1, shapeName=dyn_shorts_r_01_capsule, rotation=<rot -90 0 0>}]
 dyn_shorts_fmd_r_01 [{shapeType=Capsule, length=6, radius=2.25, shapeName=dyn_shorts_r_01_capsule, center=<4.5 0 0.2787>, rotation=<rot 90 0 0>}]
 dyn_shorts_fmd_r_02 [{shapeType=Capsule, length=6, radius=2.5, shapeName=dyn_shorts_r_01_capsule, center=<4.5 0 0>, rotation=<rot 90 0 0>}]
 dyn_shorts_fmd_r_03 [{shapeType=Capsule, length=6, radius=2.5, shapeName=dyn_shorts_r_01_capsule, center=<4.5 0 0>, rotation=<rot 90 0 0>}]
 dyn_shorts_fmd_r_04 [{shapeType=Capsule, length=6, radius=2.5, shapeName=dyn_shorts_r_01_capsule, center=<4.5 0 0>, rotation=<rot 90 0 0>}]
 thigh_r [{shapeType=Capsule, length=24.1873, radius=7, shapeName=thigh_r_capsule, center=<21.0789 0.4874 0.6823>, rotation=<rot -0.1947 -94.3984 90>}]
}
constraints [
 {
  bone1Name=base_shorts_fmd_r_01
  bone2Name=thigh_r
  swing1LimitDegrees=45
  swing1Motion=Free
  swing2LimitDegrees=45
  swing2Motion=Free
  twistLimitDegrees=45
  twistMotion=Free
 }
 {bone1Name=dyn_shorts_fmd_r_01, bone2Name=base_shorts_fmd_r_01, swing1LimitDegrees=12, swing2LimitDegrees=12, twistLimitDegrees=0}
 {
  bone1Name=base_shorts_fmd_r_02
  bone2Name=thigh_r
  swing1LimitDegrees=45
  swing1Motion=Free
  swing2LimitDegrees=45
  swing2Motion=Free
  twistLimitDegrees=45
  twistMotion=Free
 }
 {bone1Name=dyn_shorts_fmd_r_02, bone2Name=base_shorts_fmd_r_02, swing1LimitDegrees=6, swing2LimitDegrees=12, twistLimitDegrees=0}
 {
  bone1Name=base_shorts_fmd_r_03
  bone2Name=thigh_r
  swing1LimitDegrees=45
  swing1Motion=Free
  swing2LimitDegrees=45
  swing2Motion=Free
  twistLimitDegrees=45
  twistMotion=Free
 }
 {bone1Name=dyn_shorts_fmd_r_03, bone2Name=base_shorts_fmd_r_03, swing1LimitDegrees=12, swing2LimitDegrees=12, twistLimitDegrees=0}
 {
  bone1Name=base_shorts_fmd_r_04
  bone2Name=thigh_r
  swing1LimitDegrees=45
  swing1Motion=Free
  swing2LimitDegrees=45
  swing2Motion=Free
  twistLimitDegrees=45
  twistMotion=Free
 }
 {bone1Name=dyn_shorts_fmd_r_04, bone2Name=base_shorts_fmd_r_04, swing1LimitDegrees=12, swing2LimitDegrees=6, twistLimitDegrees=0}
]
