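# Control Rig for the FeatherMudLounge Might face accessory, correcting hair bones as the head moves.
# Weighted parent constraints spread lip, cheek and jaw-mounted front hair bones across face, neck, spine and clavicles.
# Four spherical pose readers on the head weight offsets to front, side and back hair bases and clavicle constraints.
# Skydiving and emote variables blend some offsets; frontEnd adds a right-front tweak, while the head-control function is unwired.
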ControlRigBlueprint /BRCosmetics/Characters/Player/Male/Medium/Bodies/M_MED_FeatherMudLounge_Might/Meshes/Parts/M_MED_FeatherMudLounge_Might_FaceAcc_CtrlRig
graphs [
 New Function_ContainedGraph {nodes [
   RigVMFunctionEntryNode()
   > &RigUnit_SetTransform(
    Item=(Type=Bone,Name="head")
    Weight=1
    bInitial=False
    Value=&RigUnit_GetTransform(Item=(Type=Control,Name="head_ctrl"), bInitial=False).Transform
   )
   > RigVMFunctionReturnNode()
  ]}
 RigVMFunctionLibrary {nodes [controls = RigVMCollapseNode(ExecuteContext=())]}
 RigVMModel {nodes [
   RigUnit_BeginExecution = &RigUnit_BeginExecution()
   RigVMFunction_Sequence = &RigVMFunction_Sequence(ExecuteContext=@RigUnit_BeginExecution.ExecutePin)
   &RigUnit_ParentConstraint(
    AdvancedSettings=(RotationOrderForFilter=XZY)
    Child=(Type=Bone,Name="R_lip_lower_outer")
    Parents="((Item=(Type=Bone,Name=\"R_cheek_inner\"),Weight=0.0004),(Item=(Type=Bone,Name=\"R_lip_corner\"),Weight=0.0008),(Item=(Type=Bone,Name=\"head\"),Weight=0.0283),(Item=(Type=Bone,Name=\"C_lip_lower_mid\"),Weight=0.03),(Item=(Type=Bone,Name=\"C_jaw\"),Weight=0.3744),(It ...(truncated)"
    Weight=1
    ExecutePin=@RigVMFunction_Sequence.B
   )
   > &RigUnit_ParentConstraint(
    AdvancedSettings=(RotationOrderForFilter=XZY)
    Child=(Type=Bone,Name="L_lip_lower_outer")
    Parents="((Item=(Type=Bone,Name=\"L_cheek_inner\"),Weight=0.0004),(Item=(Type=Bone,Name=\"L_lip_corner\"),Weight=0.0008),(Item=(Type=Bone,Name=\"head\"),Weight=0.0283),(Item=(Type=Bone,Name=\"C_lip_lower_mid\"),Weight=0.03),(Item=(Type=Bone,Name=\"C_jaw\"),Weight=0.3744),(It ...(truncated)"
    Weight=1
   )
   > &RigUnit_ParentConstraint(
    AdvancedSettings=(RotationOrderForFilter=XZY)
    Child=(Type=Bone,Name="C_lip_lower_mid")
    Parents=((Item=(Type=Bone,Name="R_lip_lower_outer"),Weight=0.0086),(Item=(Type=Bone,Name="L_lip_lower_outer"),Weight=0.0127),(Item=(Type=Bone,Name="C_jaw"),Weight=0.0816),(Item=(Type=Bone,Name="C_lip_lower_mid"),Weight=0.8971))
    Weight=1
   )
   > &RigUnit_ParentConstraint(
    AdvancedSettings=(RotationOrderForFilter=XZY)
    Child=(Type=Bone,Name="L_cheek_inner")
    Parents=((Item=(Type=Bone,Name="head"),Weight=0.0021),(Item=(Type=Bone,Name="L_cheek_inner"),Weight=0.9979))
    Weight=1
   )
   > &RigUnit_ParentConstraint(
    AdvancedSettings=(RotationOrderForFilter=XZY)
    Child=(Type=Bone,Name="R_cheek_inner")
    Parents=((Item=(Type=Bone,Name="head"),Weight=0.0021),(Item=(Type=Bone,Name="R_cheek_inner"),Weight=0.9979))
    Weight=1
   )
   > &RigUnit_ParentConstraint(
    AdvancedSettings=(RotationOrderForFilter=XZY)
    Child=(Type=Bone,Name="L_lip_corner")
    Parents="((Item=(Type=Bone,Name=\"L_cheek_inner\"),Weight=0.0002),(Item=(Type=Bone,Name=\"L_lip_lower_outer\"),Weight=0.0032),(Item=(Type=Bone,Name=\"head\"),Weight=0.0072),(Item=(Type=Bone,Name=\"C_jaw\"),Weight=0.0107),(Item=(Type=Bone,Name=\"L_lip_corner\"),Weight=0.9788) ...(truncated)"
    Weight=1
   )
   > &RigUnit_ParentConstraint(
    AdvancedSettings=(RotationOrderForFilter=XZY)
    Child=(Type=Bone,Name="R_lip_corner")
    Parents="((Item=(Type=Bone,Name=\"R_cheek_inner\"),Weight=0.0002),(Item=(Type=Bone,Name=\"R_lip_lower_outer\"),Weight=0.0032),(Item=(Type=Bone,Name=\"head\"),Weight=0.0072),(Item=(Type=Bone,Name=\"C_jaw\"),Weight=0.0107),(Item=(Type=Bone,Name=\"R_lip_corner\"),Weight=0.9788) ...(truncated)"
    Weight=1
   )
   > &RigUnit_ParentConstraint(
    AdvancedSettings=(RotationOrderForFilter=XZY)
    Child=(Type=Bone,Name="dyn_hair_fmd_fr_r_01")
    Parents=((Item=(Type=Bone,Name="R_cheek_inner"),Weight=0.0259),(Item=(Type=Bone,Name="head"),Weight=0.4756),(Item=(Type=Bone,Name="C_jaw"),Weight=0.4985))
    Weight=1
   )
   > &RigUnit_ParentConstraint(
    AdvancedSettings=(RotationOrderForFilter=XZY)
    Child=(Type=Bone,Name="dyn_hair_fmd_fr_r_02")
    Parents=((Item=(Type=Bone,Name="clavicle_r"),Weight=0.0356),(Item=(Type=Bone,Name="head"),Weight=0.051),(Item=(Type=Bone,Name="neck_01"),Weight=0.0718),(Item=(Type=Bone,Name="neck_02"),Weight=0.1981),(Item=(Type=Bone,Name="spine_05"),Weight=0.6434))
    Weight=1
   )
   > &RigUnit_ParentConstraint(
    AdvancedSettings=(RotationOrderForFilter=XZY)
    Child=(Type=Bone,Name="dyn_hair_fmd_fr_l_02")
    Parents=((Item=(Type=Bone,Name="clavicle_l"),Weight=0.0356),(Item=(Type=Bone,Name="head"),Weight=0.051),(Item=(Type=Bone,Name="neck_01"),Weight=0.0718),(Item=(Type=Bone,Name="neck_02"),Weight=0.1981),(Item=(Type=Bone,Name="spine_05"),Weight=0.6434))
    Weight=1
   )
   > &RigUnit_ParentConstraint(
    AdvancedSettings=(RotationOrderForFilter=XZY)
    Child=(Type=Bone,Name="dyn_hair_fmd_fr_l_01")
    Parents=((Item=(Type=Bone,Name="L_cheek_inner"),Weight=0.0259),(Item=(Type=Bone,Name="head"),Weight=0.4756),(Item=(Type=Bone,Name="C_jaw"),Weight=0.4985))
    Weight=1
   )
   > &RigUnit_ParentConstraint(
    AdvancedSettings=(RotationOrderForFilter=XZY)
    Child=(Type=Bone,Name="dyn_hair_fmd_fr_02")
    Parents="((Item=(Type=Bone,Name=\"neck_01\"),Weight=0.0118),(Item=(Type=Bone,Name=\"spine_02\"),Weight=0.0166),(Item=(Type=Bone,Name=\"clavicle_r\"),Weight=0.0196),(Item=(Type=Bone,Name=\"clavicle_l\"),Weight=0.0255),(Item=(Type=Bone,Name=\"neck_02\"),Weight=0.0269),(Item=(T ...(truncated)"
    Weight=1
   )
   > &RigUnit_ParentConstraint(
    AdvancedSettings=(RotationOrderForFilter=XZY)
    Child=(Type=Bone,Name="dyn_hair_fmd_fr_01")
    Parents="((Item=(Type=Bone,Name=\"L_cheek_inner\"),Weight=0.0002),(Item=(Type=Bone,Name=\"neck_01\"),Weight=0.0327),(Item=(Type=Bone,Name=\"neck_02\"),Weight=0.0981),(Item=(Type=Bone,Name=\"spine_05\"),Weight=0.1862),(Item=(Type=Bone,Name=\"head\"),Weight=0.3261),(Item=(Type ...(truncated)"
    Weight=1
   )
   > n1 = &RigUnit_SphericalPoseReader(
    ActiveRegionScaleFactors=(PositiveWidth=1,NegativeWidth=-0.5,PositiveHeight=1.5,NegativeHeight=1.5)
    ActiveRegionSize=0.1
    Debug=(bDrawDebug=false,bDraw2D=false,bDrawLocalAxes=false,DebugScale=25,DebugSegments=20,DebugThickness=0.25)
    DriverAxis=(X=0,Y=1,Z=0)
    DriverItem=(Type=Bone,Name="head")
    FalloffRegionScaleFactors=(PositiveWidth=3,NegativeWidth=1.8,PositiveHeight=2.5,NegativeHeight=2.5)
    FalloffSize=0.2
    FlipHeightScaling=False
    FlipWidthScaling=False
    OptionalParentItem=(Type=Bone,Name="None")
    RotationOffset=(X=90,Y=0,Z=-60)
   )
   > &RigUnit_ModifyTransforms(
    ItemToModify=((Item=(Type=Bone,Name="base_hair_fmd_back_01"),Transform=(Rotation=(X=0,Y=-0.5,Z=0,W=0.866),Translation=(X=0,Y=0,Z=0),Scale3D=(X=1,Y=1,Z=1))))
    WeightMinimum=0
    Weight=@n1.OutputParam
   )
   > &RigUnit_ModifyTransforms(
    ItemToModify=((Item=(Type=Bone,Name="base_hair_fmd_back_00_l"),Transform=(Rotation=(X=0,Y=-0.5,Z=0,W=0.866),Translation=(X=0,Y=0,Z=0),Scale3D=(X=1,Y=1,Z=1))))
    WeightMinimum=0
    Weight=@n1.OutputParam
   )
   > &RigUnit_ModifyTransforms(
    ItemToModify=((Item=(Type=Bone,Name="base_hair_fmd_back_00_r"),Transform=(Rotation=(X=0,Y=-0.7071,Z=0,W=0.7071),Translation=(X=0,Y=0,Z=-2),Scale3D=(X=1,Y=1,Z=1))))
    WeightMinimum=0
    Weight=@n1.OutputParam
   )
   n2 = /Script/RigVM.RigVMFunction_MathTransformLerp(
    A=(Rotation=(X=0,Y=-0.4226,Z=0,W=0.9063),Translation=(X=0,Y=0,Z=0),Scale3D=(X=1,Y=1,Z=1))
    B=(Rotation=(X=0,Y=-0.766,Z=0,W=0.6428),Translation=(X=-5,Y=0,Z=0),Scale3D=(X=1,Y=1,Z=1))
    T=RigVMVariableNode() {variable=skydiving}.Value
   ) {is=struct RigVMUnitNode, template="Interpolate::Execute(in A,in B,in T,out Result)"}
   > &RigUnit_ModifyTransforms(
    ItemToModify=((Item=(Type=Bone,Name="base_hair_fmd_side_l_01"),Transform=(Rotation=(X=0,Y=0,Z=0,W=1),Translation=(X=0,Y=0,Z=0),Scale3D=(X=1,Y=1,Z=1))))
    WeightMinimum=0
    ItemToModify.0.Transform=&RigVMFunction_MathTransformLerp(
     B=(Rotation=(X=0,Y=-0.5,Z=0,W=0.866),Translation=(X=-5,Y=0,Z=0),Scale3D=(X=1,Y=1,Z=1))
     A=@n2.Result
     T=RigVMVariableNode() {variable=playingEmote}.Value
    ) {template="Interpolate::Execute(in A,in B,in T,out Result)"}.Result
    Weight=@n1.OutputParam
   )
   n3 = &RigVMFunction_MathTransformLerp(
    A=(Rotation=(X=-0.2706,Y=-0.6533,Z=-0.2706,W=0.6533),Translation=(X=-10,Y=4,Z=0),Scale3D=(X=1,Y=1,Z=1))
    B=(Rotation=(X=0,Y=-0.5,Z=0,W=0.866),Translation=(X=-5,Y=0,Z=0),Scale3D=(X=1,Y=1,Z=1))
    T=RigVMVariableNode() {variable=playingEmote}.Value
   ) {template="Interpolate::Execute(in A,in B,in T,out Result)"}
   > &RigUnit_ModifyTransforms(
    ItemToModify=((Item=(Type=Bone,Name="base_hair_fmd_side_r_01"),Transform=(Rotation=(X=0.2604,Y=-0.8259,Z=0.1504,W=0.4769),Translation=(X=-10,Y=0,Z=0),Scale3D=(X=1,Y=1,Z=1))))
    WeightMinimum=0
    ItemToModify.0.Transform=&RigVMFunction_MathTransformLerp(
     B=(Rotation=(X=0,Y=-0.766,Z=0,W=0.6428),Translation=(X=-10,Y=0,Z=0),Scale3D=(X=1,Y=1,Z=1))
     A=@n3.Result
     T=RigVMVariableNode() {variable=skydiving}.Value
    ) {template="Interpolate::Execute(in A,in B,in T,out Result)"}.Result
    Weight=@n1.OutputParam
   )
   > &RigUnit_ModifyTransforms(
    ItemToModify=((Item=(Type=Bone,Name="base_hair_fmd_fr_r_01"),Transform=(Rotation=(X=0,Y=0,Z=-0.1305,W=0.9914),Translation=(X=0,Y=-8,Z=0),Scale3D=(X=1,Y=1,Z=1))))
    WeightMinimum=0
    Weight=@n1.OutputParam
   )
   > &RigUnit_ModifyTransforms(
    ItemToModify=((Item=(Type=Bone,Name="base_hair_fmd_fr_l_01"),Transform=(Rotation=(X=0,Y=0,Z=0.1305,W=0.9914),Translation=(X=0,Y=8,Z=0),Scale3D=(X=1,Y=1,Z=1))))
    WeightMinimum=0
    Weight=@n1.OutputParam
   )
   > &RigUnit_ModifyTransforms(
    ItemToModify=((Item=(Type=Bone,Name="base_hair_fmd_fr_01"),Transform=(Rotation=(X=0,Y=-0.3827,Z=0,W=0.9239),Translation=(X=-5,Y=0,Z=0),Scale3D=(X=1,Y=1,Z=1))))
    WeightMinimum=0
    Weight=@n1.OutputParam
   )
   > n4 = &RigUnit_SphericalPoseReader(
    ActiveRegionScaleFactors=(PositiveWidth=1,NegativeWidth=0.2,PositiveHeight=1.5,NegativeHeight=1.5)
    ActiveRegionSize=0.1
    Debug=(bDrawDebug=true,bDraw2D=false,bDrawLocalAxes=true,DebugScale=25,DebugSegments=20,DebugThickness=0.25)
    DriverAxis=(X=0,Y=1,Z=0)
    DriverItem=(Type=Bone,Name="head")
    FalloffRegionScaleFactors=(PositiveWidth=3,NegativeWidth=1.8,PositiveHeight=2.5,NegativeHeight=2.5)
    FalloffSize=0.2
    FlipHeightScaling=False
    FlipWidthScaling=False
    OptionalParentItem=(Type=Bone,Name="None")
    RotationOffset=(X=90,Y=0,Z=20)
   )
   n5 = RigVMVariableNode() {variable=skydiving}
   n6 = &RigVMFunction_MathTransformLerp(
    A=(Rotation=(X=-0.2164,Y=0,Z=0,W=0.9763),Translation=(X=-1,Y=0,Z=2),Scale3D=(X=1,Y=1,Z=1))
    B=(Rotation=(X=0,Y=-0.3007,Z=0,W=0.9537),Translation=(X=-3,Y=0,Z=2),Scale3D=(X=1,Y=1,Z=1))
    T=@n5.Value
   ) {template="Interpolate::Execute(in A,in B,in T,out Result)"}
   > &RigUnit_ModifyTransforms(
    ItemToModify=((Item=(Type=Bone,Name="base_hair_fmd_fr_r_01"),Transform=(Rotation=(X=0,Y=0,Z=0,W=1),Translation=(X=-3,Y=0,Z=2),Scale3D=(X=1,Y=1,Z=1))))
    WeightMinimum=0
    ItemToModify.0.Transform=@n6.Result
    Weight=@n4.OutputParam
   )
   n7 = &RigVMFunction_MathTransformLerp(
    A=(Rotation=(X=0,Y=0,Z=0,W=1),Translation=(X=-1,Y=0,Z=2),Scale3D=(X=1,Y=1,Z=1))
    B=(Rotation=(X=0,Y=-0.3007,Z=0,W=0.9537),Translation=(X=-3,Y=0,Z=2),Scale3D=(X=1,Y=1,Z=1))
    T=@n5.Value
   ) {template="Interpolate::Execute(in A,in B,in T,out Result)"}
   > &RigUnit_ModifyTransforms(
    ItemToModify=((Item=(Type=Bone,Name="base_hair_fmd_fr_l_01"),Transform=(Rotation=(X=0,Y=0,Z=0,W=1),Translation=(X=-3,Y=0,Z=2),Scale3D=(X=1,Y=1,Z=1))))
    WeightMinimum=0
    ItemToModify.0.Transform=@n7.Result
    Weight=@n4.OutputParam
   )
   > &RigUnit_ModifyTransforms(
    ItemToModify=((Item=(Type=Bone,Name="base_hair_fmd_fr_01"),Transform=(Rotation=(X=0,Y=-0.0436,Z=0,W=0.999),Translation=(X=0,Y=0,Z=0),Scale3D=(X=1,Y=1,Z=1))))
    WeightMinimum=0
    Weight=@n4.OutputParam
   )
   > &RigUnit_ModifyTransforms(
    ItemToModify=((Item=(Type=Bone,Name="base_hair_fmd_back_01"),Transform=(Rotation=(X=0,Y=0.0872,Z=0,W=0.9962),Translation=(X=0,Y=0,Z=0),Scale3D=(X=1,Y=1,Z=1))))
    WeightMinimum=0
    Weight=@n4.OutputParam
   )
   > &RigUnit_ModifyTransforms(
    ItemToModify=((Item=(Type=Bone,Name="base_hair_fmd_back_00_l"),Transform=(Rotation=(X=0,Y=0.0872,Z=0,W=0.9962),Translation=(X=0,Y=0,Z=0),Scale3D=(X=1,Y=1,Z=1))))
    WeightMinimum=0
    Weight=@n4.OutputParam
   )
   > &RigUnit_ModifyTransforms(
    ItemToModify=((Item=(Type=Bone,Name="base_hair_fmd_back_00_r"),Transform=(Rotation=(X=0,Y=0,Z=0,W=1),Translation=(X=0,Y=0,Z=0),Scale3D=(X=1,Y=1,Z=1))))
    WeightMinimum=0
    Weight=@n4.OutputParam
   )
   > &RigUnit_ModifyTransforms(
    ItemToModify=((Item=(Type=Bone,Name="base_hair_fmd_side_l_01"),Transform=(Rotation=(X=0,Y=-0.0436,Z=0,W=0.999),Translation=(X=0,Y=0,Z=0),Scale3D=(X=1,Y=1,Z=1))))
    Weight=1
    WeightMinimum=0
   )
   > &RigUnit_ModifyTransforms(
    ItemToModify=((Item=(Type=Bone,Name="base_hair_fmd_side_r_01"),Transform=(Rotation=(X=0,Y=-0.0436,Z=0,W=0.999),Translation=(X=0,Y=0,Z=0),Scale3D=(X=1,Y=1,Z=1))))
    WeightMinimum=0
    Weight=@n4.OutputParam
   )
   > n8 = &RigUnit_SphericalPoseReader(
    ActiveRegionScaleFactors=(PositiveWidth=1,NegativeWidth=1,PositiveHeight=1.5,NegativeHeight=2.2)
    ActiveRegionSize=0.1
    Debug=(bDrawDebug=true,bDraw2D=false,bDrawLocalAxes=false,DebugScale=25,DebugSegments=20,DebugThickness=0.25)
    DriverAxis=(X=0,Y=1,Z=0)
    DriverItem=(Type=Bone,Name="head")
    FalloffRegionScaleFactors=(PositiveWidth=3,NegativeWidth=1,PositiveHeight=2.5,NegativeHeight=2.5)
    FalloffSize=0.2
    FlipHeightScaling=False
    FlipWidthScaling=False
    OptionalParentItem=(Type=Bone,Name="None")
    RotationOffset=(X=40,Y=0,Z=0)
   )
   > n9 = &RigUnit_SphericalPoseReader(
    ActiveRegionScaleFactors=(PositiveWidth=1,NegativeWidth=1,PositiveHeight=2.2,NegativeHeight=1.5)
    ActiveRegionSize=0.1
    Debug=(bDrawDebug=true,bDraw2D=false,bDrawLocalAxes=false,DebugScale=25,DebugSegments=20,DebugThickness=0.25)
    DriverAxis=(X=0,Y=1,Z=0)
    DriverItem=(Type=Bone,Name="head")
    FalloffRegionScaleFactors=(PositiveWidth=3,NegativeWidth=1,PositiveHeight=2.5,NegativeHeight=2.5)
    FalloffSize=0.2
    FlipHeightScaling=False
    FlipWidthScaling=False
    OptionalParentItem=(Type=Bone,Name="None")
    RotationOffset=(X=140,Y=0,Z=0)
   )
   > &RigUnit_ParentConstraint(
    AdvancedSettings=(InterpolationType=Average,RotationOrderForFilter=XZY)
    Child=(Type=Bone,Name="dyn_hair_fmd_fr_02")
    Parents=((Item=(Type=Bone,Name="spine_05"),Weight=0.5),(Item=(Type=Bone,Name="neck_01"),Weight=0.5))
    Weight=1
    bMaintainOffset=true
   )
   > &RigUnit_ParentConstraint(
    AdvancedSettings=(InterpolationType=Average,RotationOrderForFilter=XZY)
    Child=(Type=Bone,Name="dyn_hair_fmd_fr_l_02")
    Parents=((Item=(Type=Bone,Name="clavicle_l"),Weight=0.5),(Item=(Type=Bone,Name="base_hair_fmd_side_l_01"),Weight=0.5))
    Weight=@n8.OutputParam
   )
   > &RigUnit_ParentConstraint(
    AdvancedSettings=(InterpolationType=Average,RotationOrderForFilter=XZY)
    Child=(Type=Bone,Name="dyn_hair_fmd_fr_r_02")
    Parents=((Item=(Type=Bone,Name="clavicle_r"),Weight=0.5),(Item=(Type=Bone,Name="base_hair_fmd_side_r_01"),Weight=0.5))
    Weight=@n9.OutputParam
   )
   > &RigUnit_ModifyTransforms(
    ItemToModify="((Item=(Type=Bone,Name=\"None\"),Transform=(Rotation=(X=0,Y=0,Z=0,W=1),Translation=(X=0,Y=0,Z=0),Scale3D=(X=1,Y=1,Z=1))),(Item=(Type=Bone,Name=\"base_hair_fmd_fr_01\"),Transform=(Rotation=(X=0,Y=-0.0872,Z=0,W=0.9962),Translation=(X=-3,Y=0,Z=0),Scale3D=(X=1,Y=1 ...(truncated)"
    Weight=1
    WeightMinimum=0
   )
   > &RigUnit_ModifyTransforms(
    ItemToModify=((Item=(Type=Bone,Name="None"),Transform=(Rotation=(X=0,Y=0,Z=0,W=1),Translation=(X=0,Y=0,Z=0),Scale3D=(X=1,Y=1,Z=1))),(Item=(Type=Bone,Name="base_hair_fmd_fr_l_01"),Transform=(Rotation=(X=0,Y=0,Z=0,W=1),Translation=(X=-3,Y=0,Z=0),Scale3D=(X=1,Y=1,Z=1))))
    WeightMinimum=0
    Weight=@n8.OutputParam
   )
   > &RigUnit_ModifyTransforms(
    ItemToModify=((Item=(Type=Bone,Name="None"),Transform=(Rotation=(X=0,Y=0,Z=0,W=1),Translation=(X=0,Y=0,Z=0),Scale3D=(X=1,Y=1,Z=1))),(Item=(Type=Bone,Name="base_hair_fmd_fr_r_01"),Transform=(Rotation=(X=0,Y=0,Z=0,W=1),Translation=(X=-3,Y=0,Z=0),Scale3D=(X=1,Y=1,Z=1))))
    WeightMinimum=0
    Weight=@n9.OutputParam
   )
   Interpolate = /Script/RigVM.RigVMFunction_MathDoubleLerp(A=1, B=0, T=RigVMVariableNode() {variable=skydiving}.Value) {is=struct RigVMUnitNode, template="Interpolate::Execute(in A,in B,in T,out Result)"}
   > &RigUnit_ModifyTransforms(
    ItemToModify=((Item=(Type=Bone,Name="base_hair_fmd_side_l_01"),Transform=(Rotation=(X=0,Y=0.0872,Z=0,W=0.9962),Translation=(X=0,Y=0,Z=0),Scale3D=(X=1,Y=1,Z=1))))
    WeightMinimum=0
    Weight=@Interpolate.Result
   )
   > n10 = &RigUnit_ModifyTransforms(
    ItemToModify=((Item=(Type=Bone,Name="base_hair_fmd_fr_01"),Transform=(Rotation=(X=0,Y=-0.0872,Z=0,W=0.9962),Translation=(X=0,Y=0,Z=-5),Scale3D=(X=1,Y=1,Z=1))))
    Weight=1
    WeightMinimum=0
   )
   RigVMFunction_ControlFlowBranch = &RigVMFunction_ControlFlowBranch(ExecuteContext=@n10.ExecutePin, Condition=RigVMVariableNode() {variable=frontEnd}.Value)
   &RigUnit_ModifyTransforms(
    ItemToModify=((Item=(Type=Bone,Name="base_hair_fmd_fr_r_01"),Transform=(Rotation=(X=0,Y=-0.0872,Z=0,W=0.9962),Translation=(X=0,Y=0,Z=1),Scale3D=(X=1,Y=1,Z=1))))
    WeightMinimum=0
    ExecutePin=@RigVMFunction_ControlFlowBranch.True
    Weight=@n9.OutputParam
   )
   &RigUnit_ParentConstraint(
    AdvancedSettings=(InterpolationType=Average,RotationOrderForFilter=XZY)
    Child=(Type=Bone,Name="dyn_hair_fmd_fr_r_02")
    ExecutePin=()
    Parents=((Item=(Type=Bone,Name="clavicle_r"),Weight=0.5),(Item=(Type=Bone,Name="base_hair_fmd_side_r_01"),Weight=0.5))
    Weight=1
   )
   controls = RigVMFunctionReferenceNode(ExecuteContext=())
  ]}
]
hierarchy [
 root {loc=<0 0 0>, rot=<0 0 0 1>}
 pelvis {parent=root, loc=<0 -0.7071 93.1191>, rot=<0.0127 0.707 -0.0127 -0.707>}
 spine_01 {parent=pelvis, loc=<6.7494 -0.6388 0>, rot=<0 0 0.034 -0.9994>}
 spine_02 {parent=spine_01, loc=<7.252 -0.035 0>, rot=<0 0 0.0238 -0.9997>}
 spine_03 {parent=spine_02, loc=<6.0106 0 0>, rot=<0 0 0.0149 -0.9999>}
 spine_04 {parent=spine_03, loc=<8.1024 0 0>, rot=<0 0 -0.0132 -0.9999>}
 spine_05 {parent=spine_04, loc=<10.0142 0 0>, rot=<0 0 -0.0015 -1>}
 clavicle_l {parent=spine_05, loc=<4.3575 -0.1466 -2.3223>, rot=<0.0465 -0.7289 -0.009 -0.6829>}
 upperarm_l {parent=clavicle_l, loc=<15.5111 0 0>, rot=<0.0226 -0.3234 -0.0088 -0.9459>}
 lowerarm_l {parent=upperarm_l, loc=<26.6514 0 0>, rot=<0.0391 -0.0099 -0.2453 -0.9686>}
 hand_l {parent=lowerarm_l, loc=<22.3491 0 0>, rot=<-0.7081 -0.0507 0.0975 -0.6975>}
 thumb_01_l {parent=hand_l, loc=<1.3118 -0.3191 -2.459>, rot=<0.6546 -0.4042 0.0664 -0.6354>}
 thumb_02_l {parent=thumb_01_l, loc=<5.9336 -0.0448 0.0401>, rot=<-0.0037 0.0213 0.298 -0.9543>}
 thumb_03_l {parent=thumb_02_l, loc=<2.2223 0.3682 -0.0907>, rot=<0.0118 -0.013 0.0066 -0.9998>}
 index_metacarpal_l {parent=hand_l, loc=<1.6535 0.0183 -2.96>, rot=<0.0158 -0.0124 -0.0173 -0.9996>}
 index_01_l {parent=index_metacarpal_l, loc=<8.3008 0.0357 -0.0524>, rot=<-0.0044 0.0485 0.0924 -0.9945>}
 index_02_l {parent=index_01_l, loc=<3.3587 -0.3004 0.0452>, rot=<-0.002 0.0027 0.3225 -0.9466>}
 index_03_l {parent=index_02_l, loc=<2.5578 -0.123 -0.0252>, rot=<-0.002 0.0067 0.2558 -0.9667>}
 middle_metacarpal_l {parent=hand_l, loc=<2.1856 -0.1725 -0.5984>, rot=<0.006 0.0168 -0.0329 -0.9993>}
 middle_01_l {parent=middle_metacarpal_l, loc=<7.3275 -0.2447 0.0129>, rot=<-0.0208 0.0042 0.0858 -0.9961>}
 middle_02_l {parent=middle_01_l, loc=<4.0123 0.1401 0.1175>, rot=<-0.0218 0.0258 0.3512 -0.9357>}
 middle_03_l {parent=middle_02_l, loc=<2.6077 0.0534 0.1373>, rot=<0.0082 0.0188 0.1708 -0.9851>}
 ring_metacarpal_l {parent=hand_l, loc=<2.2074 -0.3625 1.3638>, rot=<-0.0284 0.0722 -0.0458 -0.9959>}
 ring_01_l {parent=ring_metacarpal_l, loc=<7.0295 -0.6197 -0.1446>, rot=<-0.0489 -0.0252 0.1892 -0.9804>}
 ring_02_l {parent=ring_01_l, loc=<3.6174 0.4647 -0.1027>, rot=<0 0 0.2806 -0.9598>}
 ring_03_l {parent=ring_02_l, loc=<2.4152 0.114 0.0199>, rot=<0 0 0.1189 -0.9929>}
 pinky_metacarpal_l {parent=hand_l, loc=<1.7616 -0.9753 3.3814>, rot=<-0.0397 0.0814 -0.069 -0.9935>}
 pinky_01_l {parent=pinky_metacarpal_l, loc=<7.1709 -0.4831 -0.1094>, rot=<-0.0654 -0.0035 0.2242 -0.9723>}
 pinky_02_l {parent=pinky_01_l, loc=<2.8274 0.0893 -0.1043>, rot=<0.0037 0.0011 0.2854 -0.9584>}
 pinky_03_l {parent=pinky_02_l, loc=<2.2046 0.0307 -0.0078>, rot=<-0.0002 0.0098 0.1445 -0.9895>}
 deform_wrist_dn_base_l {parent=hand_l, loc=<0 -2.2342 0>, rot=<0 0 0 -1>}
 deform_wrist_dn_l {parent=deform_wrist_dn_base_l, loc=<0 0 0>, rot=<0 0 0 1>}
 deform_wrist_up_base_l {parent=hand_l, loc=<0 2.2682 0>, rot=<0 0 0 -1>}
 deform_wrist_up_l {parent=deform_wrist_up_base_l, loc=<0 0 0>, rot=<0 0 0 1>}
 lowerarm_twist_02_l {parent=lowerarm_l, loc=<9.6199 0 0>, rot=<0 0 0 -1>}
 lowerarm_twist_01_l {parent=lowerarm_l, loc=<17.8549 0 0>, rot=<0 0 0 -1>}
 deform_elbow_in_base_l {parent=lowerarm_l, loc=<0.2112 5.259 0.8274>, rot=<0 0 0 -1>}
 deform_elbow_in_l {parent=deform_elbow_in_base_l, loc=<0 0 0>, rot=<0 0 0 1>}
 deform_elbow_out_base_l {parent=lowerarm_l, loc=<0 -4 0>, rot=<0 0 0 -1>}
 deform_elbow_out_l {parent=deform_elbow_out_base_l, loc=<0 0 0>, rot=<0 0 0 1>}
 upperarm_twist_01_l {parent=upperarm_l, loc=<0.0041 0 0>, rot=<0 0 0 -1>}
 deform_upperarm_bk_base_l {parent=upperarm_twist_01_l, loc=<1.0813 -7.2193 -1.9453>, rot=<0 0 0 1>}
 deform_upperarm_bk_l {parent=deform_upperarm_bk_base_l, loc=<0 0 0>, rot=<0 0 0 1>}
 deform_upperarm_fr_base_l {parent=upperarm_twist_01_l, loc=<-0.3691 7.0506 -2.765>, rot=<0 0 0 1>}
 deform_upperarm_fr_l {parent=deform_upperarm_fr_base_l, loc=<0 0 0>, rot=<0 0 0 1>}
 deform_upperarm_up_base_l {parent=upperarm_twist_01_l, loc=<-3.2216 -1.1391 3.3311>, rot=<0 0 0 1>}
 deform_upperarm_up_l {parent=deform_upperarm_up_base_l, loc=<0 0 0>, rot=<0 0 0 1>}
 upperarm_twist_02_l {parent=upperarm_l, loc=<13.7941 0 0>, rot=<0 0 0 -1>}
 neck_01 {parent=spine_05, loc=<11.3138 0.1529 0>, rot=<0 0 -0.2003 -0.9797>}
 neck_02 {parent=neck_01, loc=<4.5627 0 0>, rot=<0 0 0.0504 -0.9987>}
 head {parent=neck_02, loc=<5.3914 0 0>, rot=<0 0 0.115 -0.9934>}
 faceAttach {parent=head, loc=<-4.146 -0.8545 0>, rot=<0.0029 -0.7071 -0.0029 -0.7071>}
 C_jaw {parent=faceAttach, loc=<-0.0134 5.7098 2.5271>, rot=<0 0 0 1>}
 base_hair_fmd_fr_01 {parent=C_jaw, loc=<0 2.1705 -6.0471>, rot=<0.4727 -0.4727 -0.5259 -0.5259>}
 dyn_hair_fmd_fr_01 {parent=base_hair_fmd_fr_01, loc=<0 0 0>, rot=<0 0 0 1>}
 dyn_hair_fmd_fr_02 {parent=dyn_hair_fmd_fr_01, loc=<6.9314 0 0>, rot=<0 0 0 1>}
 base_hair_fmd_fr_l_01 {parent=C_jaw, loc=<6.5135 0.89 -1.1901>, rot=<0.4387 -0.5188 -0.2509 -0.6895>}
 dyn_hair_fmd_fr_l_01 {parent=base_hair_fmd_fr_l_01, loc=<0 0 0>, rot=<0 0 0 1>}
 dyn_hair_fmd_fr_l_02 {parent=dyn_hair_fmd_fr_l_01, loc=<6.3054 0.6069 -1.279>, rot=<-0.1053 -0.0174 -0.0024 -0.9943>}
 base_hair_fmd_fr_r_01 {parent=C_jaw, loc=<-6.4575 0.911 -1.2383>, rot=<-0.5188 0.4387 0.6895 0.2509>}
 dyn_hair_fmd_fr_r_01 {parent=base_hair_fmd_fr_r_01, loc=<0 0 0>, rot=<0 0 0 -1>}
 dyn_hair_fmd_fr_r_02 {parent=dyn_hair_fmd_fr_r_01, loc=<6.3428 -0.2603 -1.0242>, rot=<0.0461 -0.0979 0.0121 -0.9941>}
 R_lip_lower_outer {parent=C_jaw, loc=<-2.042 7.8583 -1.9259>, rot=<0.013 -0.0445 -0.3057 -0.951>}
 L_lip_lower_outer {parent=C_jaw, loc=<2.0688 7.8583 -1.9259>, rot=<0.013 0.0445 0.3057 -0.951>}
 C_lip_lower_mid {parent=C_jaw, loc=<0.0223 8.9235 -1.1507>, rot=<0 0 0 -1>}
 L_cheek_inner {parent=faceAttach, loc=<5.2598 10.8528 5.2128>, rot=<-0.0706 -0.2697 0.3184 -0.906>}
 R_cheek_inner {parent=faceAttach, loc=<-5.2598 10.8528 5.2128>, rot=<-0.0706 0.2697 -0.3184 -0.906>}
 L_lip_corner {parent=faceAttach, loc=<3.3653 11.2742 0.8042>, rot=<0.003 0.008 -0.0146 -0.9999>}
 R_lip_corner {parent=faceAttach, loc=<-3.3653 11.2742 0.8042>, rot=<0.003 -0.008 0.0146 -0.9999>}
 base_hair_fmd_back_01 {parent=head, loc=<10.9201 -6.7511 0>, rot=<0.1295 -0.6952 0.6952 -0.1295>}
 dyn_hair_fmd_back_01 {parent=base_hair_fmd_back_01, loc=<0 0 0>, rot=<0 0 0 1>}
 dyn_hair_fmd_back_02 {parent=dyn_hair_fmd_back_01, loc=<10.7232 0 0.5741>, rot=<0 -0.0708 0 -0.9975>}
 dyn_hair_fmd_back_03 {parent=dyn_hair_fmd_back_02, loc=<11.985 0 -0.5431>, rot=<0 -0.0708 0 -0.9975>}
 dyn_hair_fmd_back_04 {parent=dyn_hair_fmd_back_03, loc=<7.3956 0 1.0445>, rot=<0 0.0126 0 -0.9999>}
 base_hair_fmd_top_r {parent=head, loc=<12.1273 8.4573 2.6401>, rot=<0.6375 -0.2324 0.365 -0.6374>}
 dyn_hair_fmd_top_r {parent=base_hair_fmd_top_r, loc=<0 0 0>, rot=<0 0 0 1>}
 base_hair_fmd_top_l {parent=head, loc=<12.1269 8.4573 -2.6401>, rot=<0.6374 -0.365 0.2324 -0.6375>}
 dyn_hair_fmd_top_l {parent=base_hair_fmd_top_l, loc=<0 0 0>, rot=<0 0 0 1>}
 base_hair_fmd_back_00_l {parent=head, loc=<10.4146 -4.7535 -6.9313>, rot=<0.1146 -0.8938 0.4071 -0.1493>}
 dyn_hair_fmd_back_01_l {parent=base_hair_fmd_back_00_l, loc=<0 0 0>, rot=<0 0 0 1>}
 dyn_hair_fmd_back_02_l {parent=dyn_hair_fmd_back_01_l, loc=<9.3768 -0.9801 0.1432>, rot=<-0.0485 -0.0975 -0.0458 -0.993>}
 dyn_hair_fmd_back_03_l {parent=dyn_hair_fmd_back_02_l, loc=<10.9374 -0.794 0.9325>, rot=<-0.1832 0.0487 -0.0276 -0.9815>}
 base_hair_fmd_back_00_r {parent=head, loc=<10.4142 -4.7535 6.9313>, rot=<0.1493 -0.4071 0.8938 -0.1146>}
 dyn_hair_fmd_back_01_r {parent=base_hair_fmd_back_00_r, loc=<0 0 0>, rot=<0 0 0 1>}
 dyn_hair_fmd_back_02_r {parent=dyn_hair_fmd_back_01_r, loc=<9.3765 0.9801 0.1433>, rot=<0.0485 -0.0975 0.0458 -0.993>}
 dyn_hair_fmd_back_03_r {parent=dyn_hair_fmd_back_02_r, loc=<10.9377 0.794 0.9325>, rot=<0.1832 0.0487 0.0276 -0.9815>}
 base_crown {parent=head, loc=<22.6794 0.4387 0.0134>, rot=<-0.7071 -0.0082 -0.0082 -0.7071>}
 dyn_crown {parent=base_crown, loc=<0 0 0>, rot=<0 0 0 1>}
 base_hair_fmd_side_l_01 {parent=head, loc=<8.3596 -1.8907 -10.6394>, rot=<0.1051 -0.8142 0.5533 -0.1407>}
 dyn_hair_fmd_side_l_01 {parent=base_hair_fmd_side_l_01, loc=<0 0 0>, rot=<0 0 0 1>}
 dyn_hair_fmd_side_l_02 {parent=dyn_hair_fmd_side_l_01, loc=<8.3292 0.1094 -0.776>, rot=<-0.0269 -0.0763 -0.0027 -0.9967>}
 dyn_hair_fmd_side_l_03 {parent=dyn_hair_fmd_side_l_02, loc=<9.5264 0 0>, rot=<0 0 0 1>}
 base_hair_fmd_side_r_01 {parent=head, loc=<8.3592 -1.8907 10.6394>, rot=<0.1291 -0.6542 0.7395 -0.0922>}
 dyn_hair_fmd_side_r_01 {parent=base_hair_fmd_side_r_01, loc=<0 0 0>, rot=<0 0 0 1>}
 dyn_hair_fmd_side_r_02 {parent=dyn_hair_fmd_side_r_01, loc=<8.3287 -0.1093 -0.7759>, rot=<0.0269 -0.0763 0.0027 -0.9967>}
 dyn_hair_fmd_side_r_03 {parent=dyn_hair_fmd_side_r_02, loc=<9.5264 0 0>, rot=<0 0 0 1>}
 clavicle_r {parent=spine_05, loc=<4.3575 -0.1466 2.3223>, rot=<0.7289 0.0465 -0.6829 0.009>}
 upperarm_r {parent=clavicle_r, loc=<-15.5111 0 0>, rot=<0.0226 -0.3234 -0.0088 -0.9459>}
 lowerarm_r {parent=upperarm_r, loc=<-26.6514 0 0>, rot=<0.0391 -0.0099 -0.2453 -0.9686>}
 hand_r {parent=lowerarm_r, loc=<-22.3491 0 0>, rot=<-0.7081 -0.0507 0.0975 -0.6975>}
 thumb_01_r {parent=hand_r, loc=<-1.3118 0.3191 2.459>, rot=<0.6546 -0.4042 0.0664 -0.6354>}
 thumb_02_r {parent=thumb_01_r, loc=<-5.9336 0.0448 -0.0401>, rot=<-0.0037 0.0213 0.298 -0.9543>}
 thumb_03_r {parent=thumb_02_r, loc=<-2.2223 -0.3682 0.0907>, rot=<0.0118 -0.013 0.0066 -0.9998>}
 index_metacarpal_r {parent=hand_r, loc=<-1.6535 -0.0183 2.96>, rot=<0.0158 -0.0124 -0.0173 -0.9996>}
 index_01_r {parent=index_metacarpal_r, loc=<-8.3008 -0.0357 0.0524>, rot=<-0.0044 0.0485 0.0924 -0.9945>}
 index_02_r {parent=index_01_r, loc=<-3.3587 0.3004 -0.0452>, rot=<-0.002 0.0027 0.3225 -0.9466>}
 index_03_r {parent=index_02_r, loc=<-2.5578 0.123 0.0252>, rot=<-0.002 0.0067 0.2558 -0.9667>}
 middle_metacarpal_r {parent=hand_r, loc=<-2.1856 0.1725 0.5984>, rot=<0.006 0.0168 -0.0329 -0.9993>}
 middle_01_r {parent=middle_metacarpal_r, loc=<-7.3275 0.2447 -0.0129>, rot=<-0.0208 0.0042 0.0858 -0.9961>}
 middle_02_r {parent=middle_01_r, loc=<-4.0123 -0.1401 -0.1175>, rot=<-0.0218 0.0258 0.3512 -0.9357>}
 middle_03_r {parent=middle_02_r, loc=<-2.6077 -0.0534 -0.1373>, rot=<0.0082 0.0188 0.1708 -0.9851>}
 ring_metacarpal_r {parent=hand_r, loc=<-2.2074 0.3625 -1.3638>, rot=<-0.0284 0.0722 -0.0458 -0.9959>}
 ring_01_r {parent=ring_metacarpal_r, loc=<-7.0295 0.6197 0.1446>, rot=<-0.0489 -0.0252 0.1892 -0.9804>}
 ring_02_r {parent=ring_01_r, loc=<-3.6174 -0.4647 0.1027>, rot=<0 0 0.2806 -0.9598>}
 ring_03_r {parent=ring_02_r, loc=<-2.4152 -0.114 -0.0199>, rot=<0 0 0.1189 -0.9929>}
 pinky_metacarpal_r {parent=hand_r, loc=<-1.7616 0.9753 -3.3814>, rot=<-0.0397 0.0814 -0.069 -0.9935>}
 pinky_01_r {parent=pinky_metacarpal_r, loc=<-7.1709 0.4831 0.1094>, rot=<-0.0654 -0.0035 0.2242 -0.9723>}
 pinky_02_r {parent=pinky_01_r, loc=<-2.8274 -0.0893 0.1043>, rot=<0.0037 0.0011 0.2854 -0.9584>}
 pinky_03_r {parent=pinky_02_r, loc=<-2.2046 -0.0307 0.0078>, rot=<-0.0002 0.0098 0.1445 -0.9895>}
 deform_wrist_dn_base_r {parent=hand_r, loc=<-0.0004 2.2345 -0.0001>, rot=<0 0 0 -1>}
 deform_wrist_dn_r {parent=deform_wrist_dn_base_r, loc=<0 0 0>, rot=<0 0 0 1>}
 deform_wrist_up_base_r {parent=hand_r, loc=<0.0003 -2.2684 0.0001>, rot=<0 0 0 -1>}
 deform_wrist_up_r {parent=deform_wrist_up_base_r, loc=<0 0 0>, rot=<0 0 0 1>}
 lowerarm_twist_02_r {parent=lowerarm_r, loc=<-9.6199 0 0>, rot=<0 0 0 -1>}
 lowerarm_twist_01_r {parent=lowerarm_r, loc=<-17.8549 0 0>, rot=<0 0 0 -1>}
 deform_elbow_in_base_r {parent=lowerarm_r, loc=<-0.2111 -5.2591 -0.8275>, rot=<0 0 0 -1>}
 deform_elbow_in_r {parent=deform_elbow_in_base_r, loc=<-0.0001 0 0.0002>, rot=<0 0 0 1>}
 deform_elbow_out_base_r {parent=lowerarm_r, loc=<0 4 -0.0001>, rot=<0 0 0 -1>}
 deform_elbow_out_r {parent=deform_elbow_out_base_r, loc=<0 0 0>, rot=<0 0 0 1>}
 upperarm_twist_01_r {parent=upperarm_r, loc=<-0.0041 0 0>, rot=<0 0 0 -1>}
 deform_upperarm_bk_base_r {parent=upperarm_twist_01_r, loc=<-1.0814 7.2192 1.9455>, rot=<0 0 0 1>}
 deform_upperarm_bk_r {parent=deform_upperarm_bk_base_r, loc=<0 0 0>, rot=<0 0 0 1>}
 deform_upperarm_fr_base_r {parent=upperarm_twist_01_r, loc=<0.3689 -7.0506 2.7652>, rot=<0 0 0 1>}
 deform_upperarm_fr_r {parent=deform_upperarm_fr_base_r, loc=<0 0 0>, rot=<0 0 0 1>}
 deform_upperarm_up_base_r {parent=upperarm_twist_01_r, loc=<3.2214 1.1391 -3.3308>, rot=<0 0 0 1>}
 deform_upperarm_up_r {parent=deform_upperarm_up_base_r, loc=<0 0 0>, rot=<0 0 0 -1>}
 upperarm_twist_02_r {parent=upperarm_r, loc=<-13.7941 0 0>, rot=<0 0 0 -1>}
 deform_pec_aim_base_l {parent=spine_04, loc=<7.8728 12.6388 -2.4123>, rot=<0.0102 -0.4602 0.3979 -0.7936>}
 deform_pec_aim_l {parent=deform_pec_aim_base_l, loc=<0 0 0>, rot=<0 0 0 1>}
 deform_pec_base_l {parent=deform_pec_aim_l, loc=<10.766 0 0>, rot=<0 0 0 1>}
 deform_pec_l {parent=deform_pec_base_l, loc=<0.055 0 0>, rot=<0 0 0 1>}
 deform_pec_aim_base_r {parent=spine_04, loc=<7.8723 12.6388 2.4123>, rot=<-0.4602 -0.0102 0.7936 0.3979>}
 deform_pec_aim_r {parent=deform_pec_aim_base_r, loc=<0 0 0>, rot=<0 0 0 1>}
 deform_pec_base_r {parent=deform_pec_aim_r, loc=<-10.7662 -0.0006 -0.0007>, rot=<0 0 0 1>}
 deform_pec_r {parent=deform_pec_base_r, loc=<0 0 0>, rot=<0 0 0 1>}
 thigh_l {parent=pelvis, loc=<-2.5693 -0.0109 -7.7501>, rot=<0.0012 0.0572 0.0155 -0.9982>}
 calf_l {parent=thigh_l, loc=<-39.0484 0 0>, rot=<0.0006 -0.0161 -0.0391 -0.9991>}
 foot_l {parent=calf_l, loc=<-42.561 0.1077 0.4292>, rot=<0.0418 -0.0455 0.0343 -0.9975>}
 ball_l {parent=foot_l, loc=<-6.1599 13.8264 0.6806>, rot=<0.0001 -0.0001 -0.7063 -0.7079>}
 calf_twist_02_l {parent=calf_l, loc=<-18.8064 -0.0967 0.0397>, rot=<0.0237 0.0095 -0.0035 -0.9997>}
 calf_twist_01_l {parent=calf_l, loc=<-35.813 -0.0967 0.0397>, rot=<0.0237 0.0095 -0.0035 -0.9997>}
 deform_knee_in_base_l {parent=calf_l, loc=<1.298 -5.2957 -0.0496>, rot=<0 0 0 -1>}
 deform_knee_in_l {parent=deform_knee_in_base_l, loc=<0 0 0>, rot=<0 0 0 1>}
 deform_knee_out_base_l {parent=calf_l, loc=<0 4 0>, rot=<0 0 0 -1>}
 deform_knee_out_l {parent=deform_knee_out_base_l, loc=<0 0 0>, rot=<0 0 0 1>}
 thigh_twist_01_l {parent=thigh_l, loc=<-21.6303 0 0>, rot=<0.0013 0 0 -1>}
 deform_groin_base_l {parent=thigh_l, loc=<-2.1908 8.1142 -1.7442>, rot=<0 0 0 -1>}
 deform_groin_l {parent=deform_groin_base_l, loc=<0 0 0>, rot=<0 0 0 1>}
 thigh_r {parent=pelvis, loc=<-2.5693 -0.0109 7.7501>, rot=<0.0572 -0.0012 0.9982 0.0155>}
 calf_r {parent=thigh_r, loc=<39.0484 0 0>, rot=<0.0006 -0.0161 -0.0391 -0.9991>}
 foot_r {parent=calf_r, loc=<42.561 -0.1077 -0.4292>, rot=<0.0418 -0.0455 0.0343 -0.9975>}
 ball_r {parent=foot_r, loc=<6.1599 -13.8264 -0.6806>, rot=<0.0001 -0.0001 -0.7063 -0.7079>}
 calf_twist_02_r {parent=calf_r, loc=<18.8064 0.0967 -0.0397>, rot=<0.0237 0.0095 -0.0035 -0.9997>}
 calf_twist_01_r {parent=calf_r, loc=<35.813 0.0967 -0.0397>, rot=<0.0237 0.0095 -0.0035 -0.9997>}
 deform_knee_in_base_r {parent=calf_r, loc=<-1.298 5.2957 0.0496>, rot=<0 0 0 -1>}
 deform_knee_in_r {parent=deform_knee_in_base_r, loc=<0 0 0>, rot=<0 0 0 1>}
 deform_knee_out_base_r {parent=calf_r, loc=<0 -4 0>, rot=<0 0 0 -1>}
 deform_knee_out_r {parent=deform_knee_out_base_r, loc=<0 0 0>, rot=<0 0 0 1>}
 thigh_twist_01_r {parent=thigh_r, loc=<21.6303 0 0>, rot=<0.0013 0 0 -1>}
 deform_groin_base_r {parent=thigh_r, loc=<2.1908 -8.1142 1.7442>, rot=<0 0 0 -1>}
 deform_groin_r {parent=deform_groin_base_r, loc=<0 0 0>, rot=<0 0 0 1>}
 deform_glute_base_l {parent=pelvis, loc=<-8.8064 -8.8222 -8.4287>, rot=<0.0275 0.0501 -0.4472 -0.8926>}
 deform_glute_l {parent=deform_glute_base_l, loc=<0 0 0>, rot=<0 0 0 1>}
 deform_glute_base_r {parent=pelvis, loc=<-8.8062 -8.8222 8.4287>, rot=<0.0501 -0.0275 0.8926 -0.4472>}
 deform_glute_r {parent=deform_glute_base_r, loc=<0 0 0>, rot=<0 0 0 1>}
 head_ctrl {controlType=EULER_TRANSFORM, type=Control, loc=<0 0 0>, rot=<0 0 0 1>}
]
variables [
 capeOrBackpack {type=bool, default=False}
 frontEnd {type=bool, default=False}
 playingEmote {type=double, default=0}
 skydiving {type=double, default=0}
 tacticalSprinting {type=double, default=0}
]
