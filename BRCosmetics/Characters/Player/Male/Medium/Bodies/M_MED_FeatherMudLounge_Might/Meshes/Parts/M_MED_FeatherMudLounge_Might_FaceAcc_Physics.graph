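# Physics asset for the hair bones on the FaceAcc part of the FeatherMudLounge_Might male medium cosmetic body.
# Capsules cover the upper spine, neck, head, clavicles, upper arms and five hair chains: back center, back left/right, side left/right.
# Each chain hangs from a head-linked base bone, with tight swing and twist limits between dynamic segments.
# Free-angular cross constraints tie neighboring back chains together and the right side chain to the back-right one.
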PhysicsAsset /BRCosmetics/Characters/Player/Male/Medium/Bodies/M_MED_FeatherMudLounge_Might/Meshes/Parts/M_MED_FeatherMudLounge_Might_FaceAcc_Physics
bodies {
 base_hair_fmd_back_00_l [{shapeType=Capsule, length=0.505, radius=0.505, shapeName=base_hair_fmd_back_00_l_capsule, center=<1 0 0>, rotation=<rot -90 -90 0>}]
 base_hair_fmd_back_00_r [{shapeType=Capsule, length=0.505, radius=0.505, shapeName=base_hair_fmd_back_00_r_capsule, center=<1 0 0>, rotation=<rot -90 -90 0>}]
 base_hair_fmd_back_01 [{shapeType=Capsule, length=0.505, radius=0.505, shapeName=base_hair_fmd_back_01_capsule, center=<1 0 0>, rotation=<rot -90 -90 0>}]
 base_hair_fmd_side_l_01 [{shapeType=Capsule, length=1, radius=1, rotation=<rot -90 0 0>}]
 base_hair_fmd_side_r_01 [{shapeType=Capsule, length=0.505, radius=0.505, shapeName=base_hair_fmd_side_r_01_capsule, rotation=<rot 0 -90 0>}]
 clavicle_l [{shapeType=Capsule, length=8.0945, radius=8.1, shapeName=clavicle_l_capsule, center=<10.0981 0 1>, rotation=<rot -94.0713 0 0>}]
 clavicle_r [{shapeType=Capsule, length=8.0945, radius=8.1, shapeName=clavicle_r_capsule, center=<-10.037 0 0>, rotation=<rot -95.1046 0 0>}]
 dyn_hair_fmd_back_01 [{shapeType=Capsule, length=2, radius=3.3, center=<6 0 1>, rotation=<rot 0 0 90>}]
 dyn_hair_fmd_back_01_l [{shapeType=Capsule, length=2, radius=3.3, center=<3 0 1>, rotation=<rot 0 0 90>}]
 dyn_hair_fmd_back_01_r [{shapeType=Capsule, length=2, radius=3.3, center=<3 0 1>, rotation=<rot 0 0 90>}]
 dyn_hair_fmd_back_02 [{shapeType=Capsule, length=3, radius=3.3, center=<3 0 1>, rotation=<rot 0 0 90>}]
 dyn_hair_fmd_back_02_l [{shapeType=Capsule, length=1, radius=3, center=<6 0 1>, rotation=<rot 0 0 90>}]
 dyn_hair_fmd_back_02_r [{shapeType=Capsule, length=2, radius=3, center=<6 0 1>, rotation=<rot 0 0 90>}]
 dyn_hair_fmd_back_03 [{shapeType=Capsule, length=5, radius=3.3, center=<4 0 0>, rotation=<rot -0.0005 0 90>}]
 dyn_hair_fmd_back_03_l [{shapeType=Capsule, length=4, radius=2.8, center=<3 0 3>, rotation=<rot 0 0 90>}]
 dyn_hair_fmd_back_03_r [{shapeType=Capsule, length=4, radius=3.3, center=<3 0 3>, rotation=<rot 0 0 90>}]
 dyn_hair_fmd_back_04 [{shapeType=Capsule, length=5, radius=3.3, center=<4 0 -0.5692>, rotation=<rot 0.0005 180 90>}]
 dyn_hair_fmd_side_l_01 [{shapeType=Capsule, length=2, radius=2.5, center=<7 0 0>, rotation=<rot 0 0 90>}]
 dyn_hair_fmd_side_l_02 [{shapeType=Capsule, length=2, radius=3, center=<5 0 1>, rotation=<rot 0 0 90>}]
 dyn_hair_fmd_side_l_03 [{shapeType=Capsule, length=2, radius=3, center=<7 0 5>, rotation=<rot 0 0 90>}]
 dyn_hair_fmd_side_r_01 [{shapeType=Capsule, length=2, radius=2.5, center=<7 0 0>, rotation=<rot 0 0 90>}]
 dyn_hair_fmd_side_r_02 [{shapeType=Capsule, length=2, radius=3, center=<3 0 0>, rotation=<rot 0 0 90>}]
 dyn_hair_fmd_side_r_03 [{shapeType=Capsule, length=2, radius=3, center=<3 0 0>, rotation=<rot 0 0 90>}]
 head [{shapeType=Capsule, length=7.1742, radius=8.5, shapeName=neck_02_capsule, center=<0 1.4575 0>, rotation=<rot 0 90 -90>}]
 neck_01 [{shapeType=Capsule, length=7.1742, radius=7.1742, shapeName=neck_01_capsule, rotation=<rot 0 90 -90>}]
 neck_02 [{shapeType=Capsule, length=7.1742, radius=7.1742, shapeName=neck_02_capsule, center=<0 2 0>, rotation=<rot 0 90 -90>}]
 spine_04 [{shapeType=Capsule, length=11.6031, radius=11, shapeName=spine_05_capsule, center=<0.9971 -2.6158 0>}]
 spine_05 [{shapeType=Capsule, length=11.6031, radius=11, shapeName=spine_05_capsule, center=<1.3938 0.1701 0>}]
 upperarm_l [{shapeType=Capsule, length=28.5752, radius=6.7469, shapeName=upperarm_l_capsule, center=<13.6654 0 0>, rotation=<rot -90 180 180>}]
 upperarm_r [{shapeType=Capsule, length=28.5752, radius=6.747, shapeName=upperarm_r_capsule, center=<-13.6654 0 0>, rotation=<rot -90 180 180>}]
}
constraints [
 {
  bone1Name=clavicle_l
  bone2Name=spine_05
  swing1LimitDegrees=45
  swing1Motion=Free
  swing2LimitDegrees=45
  swing2Motion=Free
  twistLimitDegrees=45
  twistMotion=Free
 }
 {
  bone1Name=upperarm_l
  bone2Name=clavicle_l
  swing1LimitDegrees=45
  swing1Motion=Free
  swing2LimitDegrees=45
  swing2Motion=Free
  twistLimitDegrees=45
  twistMotion=Free
 }
 {
  bone1Name=clavicle_r
  bone2Name=spine_05
  swing1LimitDegrees=45
  swing1Motion=Free
  swing2LimitDegrees=45
  swing2Motion=Free
  twistLimitDegrees=45
  twistMotion=Free
 }
 {
  bone1Name=upperarm_r
  bone2Name=clavicle_r
  swing1LimitDegrees=45
  swing1Motion=Free
  swing2LimitDegrees=45
  swing2Motion=Free
  twistLimitDegrees=45
  twistMotion=Free
 }
 {
  bone1Name=head
  bone2Name=neck_02
  swing1LimitDegrees=45
  swing1Motion=Free
  swing2LimitDegrees=45
  swing2Motion=Free
  twistLimitDegrees=45
  twistMotion=Free
 }
 {
  bone1Name=neck_02
  bone2Name=neck_01
  swing1LimitDegrees=45
  swing1Motion=Free
  swing2LimitDegrees=45
  swing2Motion=Free
  twistLimitDegrees=45
  twistMotion=Free
 }
 {
  bone1Name=neck_01
  bone2Name=spine_05
  swing1LimitDegrees=45
  swing1Motion=Free
  swing2LimitDegrees=45
  swing2Motion=Free
  twistLimitDegrees=45
  twistMotion=Free
 }
 {bone1Name=dyn_hair_fmd_side_l_02, bone2Name=dyn_hair_fmd_side_l_01, swing1LimitDegrees=5, swing2LimitDegrees=10, twistLimitDegrees=1}
 {bone1Name=dyn_hair_fmd_side_l_01, bone2Name=base_hair_fmd_side_l_01, swing1LimitDegrees=0, swing2LimitDegrees=10, twistLimitDegrees=1}
 {
  bone1Name=base_hair_fmd_side_l_01
  bone2Name=head
  swing1LimitDegrees=45
  swing1Motion=Free
  swing2LimitDegrees=45
  swing2Motion=Free
  twistLimitDegrees=45
  twistMotion=Free
 }
 {
  bone1Name=base_hair_fmd_back_01
  bone2Name=head
  swing1LimitDegrees=45
  swing1Motion=Free
  swing2LimitDegrees=45
  swing2Motion=Free
  twistLimitDegrees=45
  twistMotion=Free
 }
 {bone1Name=dyn_hair_fmd_back_01, bone2Name=base_hair_fmd_back_01, swing1LimitDegrees=15, swing2LimitDegrees=40, twistLimitDegrees=0}
 {bone1Name=dyn_hair_fmd_back_02, bone2Name=dyn_hair_fmd_back_01, swing1LimitDegrees=5, swing2LimitDegrees=40, twistLimitDegrees=1}
 {bone1Name=dyn_hair_fmd_back_03, bone2Name=dyn_hair_fmd_back_02, swing1LimitDegrees=15, swing2LimitDegrees=40, twistLimitDegrees=1}
 {
  bone1Name=base_hair_fmd_back_00_l
  bone2Name=head
  swing1LimitDegrees=45
  swing1Motion=Free
  swing2LimitDegrees=45
  swing2Motion=Free
  twistLimitDegrees=45
  twistMotion=Free
 }
 {bone1Name=dyn_hair_fmd_back_01_l, bone2Name=base_hair_fmd_back_00_l, swing1LimitDegrees=22, swing2LimitDegrees=15, twistLimitDegrees=0}
 {bone1Name=dyn_hair_fmd_back_02_l, bone2Name=dyn_hair_fmd_back_01_l, swing1LimitDegrees=22, swing2LimitDegrees=15, twistLimitDegrees=0}
 {bone1Name=dyn_hair_fmd_back_03_l, bone2Name=dyn_hair_fmd_back_02_l, swing1LimitDegrees=22, swing2LimitDegrees=15, twistLimitDegrees=1}
 {
  bone1Name=base_hair_fmd_back_00_r
  bone2Name=head
  swing1LimitDegrees=45
  swing1Motion=Free
  swing2LimitDegrees=45
  swing2Motion=Free
  twistLimitDegrees=45
  twistMotion=Free
 }
 {bone1Name=dyn_hair_fmd_back_01_r, bone2Name=base_hair_fmd_back_00_r, swing1LimitDegrees=15, swing2LimitDegrees=15, twistLimitDegrees=0}
 {bone1Name=dyn_hair_fmd_back_02_r, bone2Name=dyn_hair_fmd_back_01_r, swing1LimitDegrees=15, swing2LimitDegrees=20, twistLimitDegrees=1}
 {bone1Name=dyn_hair_fmd_back_03_r, bone2Name=dyn_hair_fmd_back_02_r, swing1LimitDegrees=15, swing2LimitDegrees=20, twistLimitDegrees=1}
 {
  bone1Name=dyn_hair_fmd_back_01_r
  bone2Name=dyn_hair_fmd_back_01
  swing1LimitDegrees=45
  swing1Motion=Free
  swing2LimitDegrees=45
  swing2Motion=Free
  twistLimitDegrees=45
  twistMotion=Free
 }
 {
  bone1Name=dyn_hair_fmd_back_01_l
  bone2Name=dyn_hair_fmd_back_01
  swing1LimitDegrees=45
  swing1Motion=Free
  swing2LimitDegrees=45
  swing2Motion=Free
  twistLimitDegrees=45
  twistMotion=Free
 }
 {
  bone1Name=dyn_hair_fmd_back_03_r
  bone2Name=dyn_hair_fmd_back_03
  swing1LimitDegrees=45
  swing1Motion=Free
  swing2LimitDegrees=45
  swing2Motion=Free
  twistLimitDegrees=45
  twistMotion=Free
 }
 {
  bone1Name=dyn_hair_fmd_back_03_l
  bone2Name=dyn_hair_fmd_back_03
  swing1LimitDegrees=45
  swing1Motion=Free
  swing2LimitDegrees=45
  swing2Motion=Free
  twistLimitDegrees=45
  twistMotion=Free
 }
 {
  bone1Name=spine_05
  bone2Name=spine_04
  swing1LimitDegrees=45
  swing1Motion=Free
  swing2LimitDegrees=45
  swing2Motion=Free
  twistLimitDegrees=45
  twistMotion=Free
 }
 {
  bone1Name=base_hair_fmd_side_r_01
  bone2Name=head
  swing1LimitDegrees=45
  swing1Motion=Free
  swing2LimitDegrees=45
  swing2Motion=Free
  twistLimitDegrees=45
  twistMotion=Free
 }
 {bone1Name=dyn_hair_fmd_side_r_01, bone2Name=base_hair_fmd_side_r_01, swing1LimitDegrees=2.5, swing2LimitDegrees=10, twistLimitDegrees=1}
 {
  bone1Name=dyn_hair_fmd_side_r_02
  bone2Name=dyn_hair_fmd_side_r_01
  swing1LimitDegrees=2.5
  swing1Motion=Locked
  swing2LimitDegrees=10
  twistLimitDegrees=0
 }
 {bone1Name=dyn_hair_fmd_back_04, bone2Name=dyn_hair_fmd_back_03, swing1LimitDegrees=20, swing2LimitDegrees=40, twistLimitDegrees=1}
 {
  bone1Name=dyn_hair_fmd_side_r_03
  bone2Name=dyn_hair_fmd_side_r_02
  swing1LimitDegrees=5
  swing1Motion=Locked
  swing2LimitDegrees=15
  twistLimitDegrees=1
 }
 {bone1Name=dyn_hair_fmd_side_l_03, bone2Name=dyn_hair_fmd_side_l_02, swing1LimitDegrees=5, swing2LimitDegrees=10, twistLimitDegrees=1}
 {
  bone1Name=dyn_hair_fmd_side_r_01
  bone2Name=dyn_hair_fmd_back_03_r
  swing1LimitDegrees=45
  swing1Motion=Free
  swing2LimitDegrees=45
  swing2Motion=Free
  twistLimitDegrees=45
  twistMotion=Free
 }
 {
  bone1Name=dyn_hair_fmd_side_r_02
  bone2Name=dyn_hair_fmd_back_03_r
  swing1LimitDegrees=45
  swing1Motion=Free
  swing2LimitDegrees=45
  swing2Motion=Free
  twistLimitDegrees=45
  twistMotion=Free
 }
]
