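# Anim Blueprint for the FeatherMudLounge Might face accessory, animating the character's hair and floating crown.
# Copied parent, head and body poses feed a Control Rig, then either RigidBody hair physics with state-driven side-hair corrections or, with backbling or cape, fixed bone constraints.
# AnimDynamics springs add top and front hair motion, and time-based sine waves bob the crown.
# EventGraph updates gravity, correction alphas and backbling flags.
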AnimBlueprint /BRCosmetics/Characters/Player/Male/Medium/Bodies/M_MED_FeatherMudLounge_Might/Meshes/Parts/M_MED_FeatherMudLounge_Might_FaceAcc_AnimBP
target_skeleton=&/M_MED_FeatherMudLounge_Might_FaceAcc_Skeleton
graphs [
 AnimGraph {nodes [
   n1 = "Transform (Modify) Bone - Bone: dyn_crown"(
    ComponentPose=Local To Component(LocalPose=Use cached pose 'dynamics_Cache'())
    Alpha=1
    Translation=Get Translate Crown() -> (Translate Crown:Vector)
    Rotation=Get FloatingCrownRot() -> (FloatingCrownRot:Rotator)
    Scale="1,1,1"
   ) {
    rotationMode=BMM_Additive
    rotationSpace=BCS_BoneSpace
    translationMode=BMM_Additive
    translationSpace=BCS_BoneSpace
    alphaBoolBlend {blendOption=Linear}
    boneToModify {boneName=dyn_crown}
   }
   Output Pose(Result=Component To Local(ComponentPose=@n1)) {layerGroup=None}
   n2 = Layered blend per bone(
    BasePose=Copy Pose From Mesh() {bCopyCurves=true, bUseAttachedParent=true}
    BlendPoses_0=Copy Pose From Mesh(SourceMeshComponent=Get Head_Mesh() -> (Head_Mesh:Skeletal Mesh Component Object Reference))
   ) {layerSetup [{branchFilters [
       {boneName=C_jaw}
       {boneName=R_lip_lower_outer}
       {boneName=L_lip_lower_outer}
       {boneName=C_lip_lower_mid}
       {boneName=L_cheek_inner}
       {boneName=R_cheek_inner}
       {boneName=L_lip_corner}
       {boneName=R_lip_corner}
      ]}]}
   n3 = Layered blend per bone(BasePose=@n2, BlendPoses_0=Copy Pose From Mesh(SourceMeshComponent=Get Body_Mesh())) {layerSetup [{branchFilters [{boneName=clavicle_l}, {boneName=clavicle_r}, {boneName=upperarm_r}, {boneName=upperarm_l}]}]}
   Save cached pose 'start_Cache'(Pose=@n3) {cacheName=start_Cache}
   n4 = "Constraint - dyn_hair_fmd_back_01 (PP)"!(
    ComponentPose=Local To Component(LocalPose=Use cached pose 'ctrlRig_cache'())
    Alpha=0.5
    ConstraintWeights_0=0.5
    ConstraintWeights_1:float=0.5
   ) -> (Pose) {
    alphaBoolBlend {blendOption=Linear}
    boneToModify {boneName=dyn_hair_fmd_back_01}
    constraintSetup [{targetBone {boneName=head}}, {targetBone {boneName=neck_02}}]
    constraintWeights [1,1]
   }
   n5 = "Constraint - dyn_hair_fmd_back_02 (PP)"!(ComponentPose=@n4, Alpha=0.5, ConstraintWeights_0=0.5, ConstraintWeights_1:float=0.5) -> (Pose) {
    alphaBoolBlend {blendOption=Linear}
    boneToModify {boneName=dyn_hair_fmd_back_02}
    constraintSetup [{targetBone {boneName=neck_01}}, {targetBone {boneName=neck_02}}]
    constraintWeights [1,1]
   }
   n6 = "Constraint - dyn_hair_fmd_back_03 (PP)"!(ComponentPose=@n5, Alpha=1, ConstraintWeights_0=0.5, ConstraintWeights_1:float=0.5) -> (Pose) {
    alphaBoolBlend {blendOption=Linear}
    boneToModify {boneName=dyn_hair_fmd_back_03}
    constraintSetup [{targetBone {boneName=neck_01}}, {targetBone {boneName=spine_05}}]
    constraintWeights [1,1]
   }
   n7 = "Constraint - dyn_hair_fmd_back_04 (PP)"!(ComponentPose=@n6, Alpha=1, ConstraintWeights_0=0.5, ConstraintWeights_1:float=0.5) -> (Pose) {
    alphaBoolBlend {blendOption=Linear}
    boneToModify {boneName=dyn_hair_fmd_back_04}
    constraintSetup [{targetBone {boneName=spine_04}}, {targetBone {boneName=spine_05}}]
    constraintWeights [1,1]
   }
   n8 = "Constraint - dyn_hair_fmd_back_01_l (PP)"!(ComponentPose=@n7, Alpha=0.5, ConstraintWeights_0=0.5, ConstraintWeights_1:float=0.5) -> (Pose) {
    alphaBoolBlend {blendOption=Linear}
    boneToModify {boneName=dyn_hair_fmd_back_01_l}
    constraintSetup [{targetBone {boneName=head}}, {targetBone {boneName=neck_02}}]
    constraintWeights [1,1]
   }
   n9 = "Constraint - dyn_hair_fmd_back_02_l (PP)"!(ComponentPose=@n8, Alpha=0.5, ConstraintWeights_0=0.5, ConstraintWeights_1:float=0.5) -> (Pose) {
    alphaBoolBlend {blendOption=Linear}
    boneToModify {boneName=dyn_hair_fmd_back_02_l}
    constraintSetup [{targetBone {boneName=neck_01}}, {targetBone {boneName=neck_02}}]
    constraintWeights [1,1]
   }
   n10 = "Constraint - dyn_hair_fmd_back_03_l (PP)"!(ComponentPose=@n9, Alpha=1, ConstraintWeights_0=0, ConstraintWeights_1:float=1) -> (Pose) {
    alphaBoolBlend {blendOption=Linear}
    boneToModify {boneName=dyn_hair_fmd_back_03_l}
    constraintSetup [{targetBone {boneName=neck_01}}, {targetBone {boneName=spine_05}}]
    constraintWeights [1,1]
   }
   n11 = "Constraint - dyn_hair_fmd_back_01_r (PP)"!(ComponentPose=@n10, Alpha=0.5, ConstraintWeights_0=0.5, ConstraintWeights_1:float=0.5) -> (Pose) {
    alphaBoolBlend {blendOption=Linear}
    boneToModify {boneName=dyn_hair_fmd_back_01_r}
    constraintSetup [{targetBone {boneName=neck_01}}, {targetBone {boneName=spine_05}}]
    constraintWeights [1,1]
   }
   n12 = "Constraint - dyn_hair_fmd_back_02_r (PP)"!(ComponentPose=@n11, Alpha=0.8, ConstraintWeights_0=0.2, ConstraintWeights_1:float=0.8) -> (Pose) {
    alphaBoolBlend {blendOption=Linear}
    boneToModify {boneName=dyn_hair_fmd_back_02_r}
    constraintSetup [{targetBone {boneName=neck_01}}, {targetBone {boneName=spine_05}}]
    constraintWeights [1,1]
   }
   n13 = "Constraint - dyn_hair_fmd_back_03_r (PP)"!(ComponentPose=@n12, Alpha=1, ConstraintWeights_0=0, ConstraintWeights_1:float=1) -> (Pose) {
    alphaBoolBlend {blendOption=Linear}
    boneToModify {boneName=dyn_hair_fmd_back_03_r}
    constraintSetup [{targetBone {boneName=neck_01}}, {targetBone {boneName=spine_05}}]
    constraintWeights [1,1]
   }
   n14 = "Constraint - dyn_hair_fmd_side_l_01 (PP)"!(ComponentPose=@n13, Alpha=0.5, ConstraintWeights_0=0.5, ConstraintWeights_1:float=0.5) -> (Pose) {
    alphaBoolBlend {blendOption=Linear}
    boneToModify {boneName=dyn_hair_fmd_side_l_01}
    constraintSetup [{targetBone {boneName=neck_01}}, {targetBone {boneName=neck_02}}]
    constraintWeights [1,1]
   }
   n15 = "Constraint - dyn_hair_fmd_side_l_02 (PP)"!(ComponentPose=@n14, Alpha=0.5, ConstraintWeights_0=0.5, ConstraintWeights_1:float=0.5) -> (Pose) {
    alphaBoolBlend {blendOption=Linear}
    boneToModify {boneName=dyn_hair_fmd_side_l_02}
    constraintSetup [{targetBone {boneName=neck_01}}, {targetBone {boneName=clavicle_l}}]
    constraintWeights [1,1]
   }
   n16 = "Constraint - dyn_hair_fmd_side_l_03 (PP)"!(ComponentPose=@n15, Alpha=0.5, ConstraintWeights_0=0, ConstraintWeights_1:float=1) -> (Pose) {
    alphaBoolBlend {blendOption=Linear}
    boneToModify {boneName=dyn_hair_fmd_side_l_03}
    constraintSetup [{targetBone {boneName=neck_02}}, {targetBone {boneName=clavicle_l}}]
    constraintWeights [1,1]
   }
   n17 = "Constraint - dyn_hair_fmd_side_r_01 (PP)"!(ComponentPose=@n16, Alpha=1, ConstraintWeights_0=0.5, ConstraintWeights_1:float=0.5) -> (Pose) {
    alphaBoolBlend {blendOption=Linear}
    boneToModify {boneName=dyn_hair_fmd_side_r_01}
    constraintSetup [{targetBone {boneName=head}}, {targetBone {boneName=neck_02}}]
    constraintWeights [1,1]
   }
   n18 = "Constraint - dyn_hair_fmd_side_r_02 (PP)"!(ComponentPose=@n17, Alpha=1, ConstraintWeights_0=0, ConstraintWeights_1:float=1) -> (Pose) {
    alphaBoolBlend {blendOption=Linear}
    boneToModify {boneName=dyn_hair_fmd_side_r_02}
    constraintSetup [{targetBone {boneName=neck_02}}, {targetBone {boneName=clavicle_r}}]
    constraintWeights [1,1]
   }
   n19 = "Constraint - dyn_hair_fmd_side_r_03 (PP)"!(ComponentPose=@n18, Alpha=1, ConstraintWeights_0=0, ConstraintWeights_1:float=1) -> (Pose) {
    alphaBoolBlend {blendOption=Linear}
    boneToModify {boneName=dyn_hair_fmd_side_r_03}
    constraintSetup [{targetBone {boneName=spine_05}}, {targetBone {boneName=clavicle_r}}]
    constraintWeights [1,1]
   }
   Save cached pose 'constraints_Cache'(Pose=Component To Local(ComponentPose=@n19)) {cacheName=constraints_Cache}
   n20 = Control Rig!(
    Source=Use cached pose 'start_Cache'()
    Alpha
    skydiving:double=Get skydiving() -> (skydiving:double)
    playingEmote:double=Get playingEmote() -> (playingEmote:double)
    tacticalSprinting:double=Get isCrounchingFloat() -> (isCrounchingFloat:double)
    frontEnd:bool=Get bIsInFrontEnd()
   ) -> (Pose) {
    alphaBoolBlend {blendOption=Linear}
    controlRigAssetReference {blueprintRigClass=&/M_MED_FeatherMudLounge_Might_FaceAcc_CtrlRig.M_MED_FeatherMudLounge_Might_FaceAcc_CtrlRig_C}
   }
   Save cached pose 'ctrlRig_cache'(Pose=@n20) {cacheName=ctrlRig_cache}
   n21 = RigidBody!(
    ComponentPose=Local To Component(LocalPose=Use cached pose 'ctrlRig_cache'())
    Alpha
    OverrideWorldGravity:Vector=Get Gravity Override()
    ExternalForce
   ) -> (Pose) {
    bClampLinearTranslationLimitToRefPose=true
    bOverrideWorldGravity=true
    cachedBoundsScale=1
    overridePhysicsAsset=&/M_MED_FeatherMudLounge_Might_FaceAcc_Physics
    simulationSpace=BaseBoneSpace
    alphaBoolBlend {blendOption=Linear}
    alphaScaleBiasClamp {interpSpeedIncreasing=1}
    baseBoneRef {boneName=neck_01}
    componentAppliedLinearAccClamp=<3000 3000 3000>
    componentLinearAccScale=<0.35 0.35 0.35>
    componentLinearVelScale=<0.35 0.35 0.35>
    overrideWorldGravity {z=-980}
    simSpaceSettings {maxAngularAcceleration=1000, maxAngularVelocity=3, maxLinearAcceleration=250, maxLinearVelocity=500, velocityScaleZ=0.1, worldAlpha=0.25}
   }
   n22 = "Constraint - dyn_hair_fmd_side_r_01 (PP)"(ComponentPose=@n21, Alpha=Get ConstraintValue1() -> (ConstraintValue1:double), ConstraintWeights_0=0.5, ConstraintWeights_1=0.5) {
    alphaBoolBlend {blendOption=Linear}
    boneToModify {boneName=dyn_hair_fmd_side_r_01}
    constraintSetup [{targetBone {boneName=head}}, {targetBone {boneName=neck_02}}]
    constraintWeights [1,1]
   }
   n23 = "Constraint - dyn_hair_fmd_side_r_02 (PP)"(ComponentPose=@n22, Alpha=Get ConstraintValue2() -> (ConstraintValue2:double), ConstraintWeights_0=0, ConstraintWeights_1=1) {
    alphaBoolBlend {blendOption=Linear}
    boneToModify {boneName=dyn_hair_fmd_side_r_02}
    constraintSetup [{targetBone {boneName=neck_02}}, {targetBone {boneName=clavicle_r}}]
    constraintWeights [1,1]
   }
   n24 = Get headUp() -> (headUp:double)
   n25 = "Constraint - dyn_hair_fmd_side_r_03 (PP)"(ComponentPose=@n23, Alpha=@n24, ConstraintWeights_0=0, ConstraintWeights_1=1) {
    alphaBoolBlend {blendOption=Linear}
    boneToModify {boneName=dyn_hair_fmd_side_r_03}
    constraintSetup [{targetBone {boneName=spine_05}}, {targetBone {boneName=clavicle_r}}]
    constraintWeights [1,1]
   }
   n26 = "Transform (Modify) Bone - Bone: base_hair_fmd_side_r_01"(ComponentPose=@n25, Alpha=@n24, Translation="-5,-5,0", Rotation="0,-45,0", Scale="1,1,1") {
    rotationMode=BMM_Additive
    rotationSpace=BCS_BoneSpace
    translationMode=BMM_Additive
    translationSpace=BCS_BoneSpace
    alphaBoolBlend {blendOption=Linear}
    boneToModify {boneName=base_hair_fmd_side_r_01}
    rotation {yaw=-45}
    translation=<-5 -5>
   }
   n27 = "Transform (Modify) Bone - Bone: dyn_hair_fmd_side_r_02"(ComponentPose=@n26, Alpha=@n24, Translation="0,0,5", Rotation="0,0,0", Scale="1,1,1") {
    rotationMode=BMM_Additive
    rotationSpace=BCS_BoneSpace
    translationMode=BMM_Additive
    translationSpace=BCS_BoneSpace
    alphaBoolBlend {blendOption=Linear}
    boneToModify {boneName=dyn_hair_fmd_side_r_02}
    translation {z=5}
   }
   n28 = Component To Local(ComponentPose=@n27)
   n29 = Use cached pose 'constraints_Cache'()
   n30 = OR Boolean(A=Get isWearingBackbling(), B=Get Is Wearing Cape() -> (Is Wearing Cape:bool))
   n31 = Anim Dynamics - Bone: dyn_hair_fmd_top_l(
    ComponentPose=Local To Component(LocalPose=Blend Poses by bool(BlendPose_0=@n29, BlendPose_1=@n28, bActiveValue=@n30) {blendType=Linear})
    Alpha=1
    ExternalForce="0,0,0"
   ) {
    angularDampingOverride=1
    angularSpringConstant=10000
    bAngularSpring=true
    bOverrideAngularDamping=true
    bUseGravityOverride=true
    gravityScale=0.35
    simulationSpace=RootRelative
    alphaBoolBlend {blendOption=Linear}
    boundBone {boneName=dyn_hair_fmd_top_l}
    componentLinearAccScale=<1 1 1>
    componentLinearVelScale=<1 1 1>
    gravityOverride {z=-980}
    physicsBodyDefinitions [{
      boundBone {boneName=dyn_hair_fmd_top_l}
      boxExtents=<45 45 45>
      constraintSetup {angularLimitsMax=<5 5 5>, angularLimitsMin=<-5 -5 -5>, angularTarget=<1 0 0>}
      localJointOffset=<2 0 0>
     }]
   }
   n32 = Anim Dynamics - Bone: dyn_hair_fmd_top_r(ComponentPose=@n31, Alpha=1, ExternalForce="0,0,0") {
    angularDampingOverride=1
    angularSpringConstant=10000
    bAngularSpring=true
    bOverrideAngularDamping=true
    bUseGravityOverride=true
    gravityScale=0.35
    simulationSpace=RootRelative
    alphaBoolBlend {blendOption=Linear}
    boundBone {boneName=dyn_hair_fmd_top_r}
    componentLinearAccScale=<1 1 1>
    componentLinearVelScale=<1 1 1>
    gravityOverride {z=-980}
    physicsBodyDefinitions [{
      boundBone {boneName=dyn_hair_fmd_top_r}
      boxExtents=<45 45 45>
      constraintSetup {angularLimitsMax=<5 5 5>, angularLimitsMin=<-5 -5 -5>, angularTarget=<1 0 0>}
      localJointOffset=<2 0 0>
     }]
   }
   n33 = Anim Dynamics - Bone: dyn_hair_fmd_fr_01(ComponentPose=@n32, Alpha=1, ExternalForce="0,0,0") {
    angularDampingOverride=1
    angularSpringConstant=10000
    bAngularSpring=true
    bOverrideAngularDamping=true
    bUseGravityOverride=true
    gravityScale=0.35
    simulationSpace=RootRelative
    alphaBoolBlend {blendOption=Linear}
    boundBone {boneName=dyn_hair_fmd_fr_01}
    componentLinearAccScale=<1 1 1>
    componentLinearVelScale=<1 1 1>
    gravityOverride {z=-980}
    physicsBodyDefinitions [{
      boundBone {boneName=dyn_hair_fmd_fr_01}
      boxExtents=<45 45 45>
      constraintSetup {angularLimitsMax=<5 5 5>, angularLimitsMin=<-5 -5 -5>, angularTarget=<1 0 0>}
      localJointOffset=<1 0 0>
     }]
   }
   n34 = Anim Dynamics - Bone: dyn_hair_fmd_fr_02(ComponentPose=@n33, Alpha=1, ExternalForce="0,0,0") {
    angularDampingOverride=1
    angularSpringConstant=10000
    bAngularSpring=true
    bOverrideAngularDamping=true
    bUseGravityOverride=true
    gravityScale=0.35
    simulationSpace=RootRelative
    alphaBoolBlend {blendOption=Linear}
    boundBone {boneName=dyn_hair_fmd_fr_02}
    componentLinearAccScale=<1 1 1>
    componentLinearVelScale=<1 1 1>
    gravityOverride {z=-980}
    physicsBodyDefinitions [{
      boundBone {boneName=dyn_hair_fmd_fr_02}
      boxExtents=<45 45 45>
      constraintSetup {angularLimitsMax=<5 5 5>, angularLimitsMin=<-5 -5 -5>, angularTarget=<1 0 0>}
      localJointOffset=<1 0 0>
     }]
   }
   n35 = Anim Dynamics - Bone: dyn_hair_fmd_fr_r_01(ComponentPose=@n34, Alpha=0.5, ExternalForce="0,0,0") {
    angularDampingOverride=1
    angularSpringConstant=10000
    bAngularSpring=true
    bOverrideAngularDamping=true
    bUseGravityOverride=true
    gravityScale=0.35
    simulationSpace=RootRelative
    alphaBoolBlend {blendOption=Linear}
    boundBone {boneName=dyn_hair_fmd_fr_r_01}
    componentLinearAccScale=<1 1 1>
    componentLinearVelScale=<1 1 1>
    gravityOverride {z=-980}
    physicsBodyDefinitions [{
      boundBone {boneName=dyn_hair_fmd_fr_r_01}
      boxExtents=<45 45 45>
      constraintSetup {angularLimitsMax=<5 5 5>, angularLimitsMin=<-5 -5 -5>, angularTarget=<1 0 0>}
      localJointOffset=<1 0 0>
     }]
   }
   n36 = Anim Dynamics - Bone: dyn_hair_fmd_fr_r_02(ComponentPose=@n35, Alpha=0.5, ExternalForce="0,0,0") {
    angularDampingOverride=1
    angularSpringConstant=10000
    bAngularSpring=true
    bOverrideAngularDamping=true
    bUseGravityOverride=true
    gravityScale=0.35
    simulationSpace=RootRelative
    alphaBoolBlend {blendOption=Linear}
    boundBone {boneName=dyn_hair_fmd_fr_r_02}
    componentLinearAccScale=<1 1 1>
    componentLinearVelScale=<1 1 1>
    gravityOverride {z=-980}
    physicsBodyDefinitions [{
      boundBone {boneName=dyn_hair_fmd_fr_r_02}
      boxExtents=<45 45 45>
      constraintSetup {angularLimitsMax=<5 5 5>, angularLimitsMin=<-5 -5 -5>, angularTarget=<1 0 0>}
      localJointOffset=<1 0 0>
     }]
   }
   n37 = Anim Dynamics - Bone: dyn_hair_fmd_fr_l_01(ComponentPose=@n36, Alpha=0.5, ExternalForce="0,0,0") {
    angularDampingOverride=1
    angularSpringConstant=10000
    bAngularSpring=true
    bOverrideAngularDamping=true
    bUseGravityOverride=true
    gravityScale=0.35
    simulationSpace=RootRelative
    alphaBoolBlend {blendOption=Linear}
    boundBone {boneName=dyn_hair_fmd_fr_l_01}
    componentLinearAccScale=<1 1 1>
    componentLinearVelScale=<1 1 1>
    gravityOverride {z=-980}
    physicsBodyDefinitions [{
      boundBone {boneName=dyn_hair_fmd_fr_l_01}
      boxExtents=<45 45 45>
      constraintSetup {angularLimitsMax=<5 5 5>, angularLimitsMin=<-5 -5 -5>, angularTarget=<1 0 0>}
      localJointOffset=<1 0 0>
     }]
   }
   n38 = Anim Dynamics - Bone: dyn_hair_fmd_fr_l_02(ComponentPose=@n37, Alpha=0.5, ExternalForce="0,0,0") {
    angularDampingOverride=1
    angularSpringConstant=10000
    bAngularSpring=true
    bOverrideAngularDamping=true
    bUseGravityOverride=true
    gravityScale=0.35
    simulationSpace=RootRelative
    alphaBoolBlend {blendOption=Linear}
    boundBone {boneName=dyn_hair_fmd_fr_l_02}
    componentLinearAccScale=<1 1 1>
    componentLinearVelScale=<1 1 1>
    gravityOverride {z=-980}
    physicsBodyDefinitions [{
      boundBone {boneName=dyn_hair_fmd_fr_l_02}
      boxExtents=<45 45 45>
      constraintSetup {angularLimitsMax=<5 5 5>, angularLimitsMin=<-5 -5 -5>, angularTarget=<1 0 0>}
      localJointOffset=<1 0 0>
     }]
   }
   Save cached pose 'dynamics_Cache'(Pose=Component To Local(ComponentPose=@n38)) {cacheName=dynamics_Cache}
  ]}
 EventGraph {nodes [
   n1 = Event BlueprintUpdateAnimation()
   > n2 = Sequence!(execute) -> (then_0, then_1, then_2:Exec, then_3:Exec)
   > n3 = Gravity Override State Function DEPRECATED(CharacterPartAnimInstance=Self-Reference(), GameTimeInSeconds=GetGameTimeInSeconds(), GravityOverrideParams=Get Gravity Override Params())
   > Set Gravity Override(Gravity Override=@n3)
   > Set skydiving(skydiving:double="To Float (Boolean)"(InBool=Get bIsSkyDiving())) -> (Output_Get:double)
   > Set playingEmote(playingEmote:double="To Float (Boolean)"(InBool=Get bIsPlayingEmote())) -> (Output_Get:double)
   > Set isCrounchingFloat(isCrounchingFloat:double="To Float (Boolean)"(InBool=Get bIsCrouching())) -> (Output_Get:double)
   n4 = Get bIsPlayingEmote()
   > n5 = Set emotePlay(emotePlay:double=FInterpTo(Current=Get emotePlay() -> (emotePlay:double), Target="To Float (Boolean)"(InBool=@n4), DeltaTime=@n1.DeltaTimeX, InterpSpeed=2)) -> (Output_Get:double)
   n6 = GetAlphaFromReferenceTransformDelta(SkeletalMeshComponent=Get Body_Mesh(), BoneName=spine_05, BaseName=head, InputRange=(X=0,Y=200), TwistAxis="0,0,1")
   > Set headUp(headUp:double=@n6) -> (Output_Get:double)
   > n7 = Branch(Condition=Get bIsSkyDiving())
   > Set ConstraintValue1(ConstraintValue1:double=0) -> (Output_Get:double)
   > Set ConstraintValue2(ConstraintValue2:double=0) -> (Output_Get:double)
   > Set headUp(headUp=0)
   n8 = Branch(execute=@n7.else, Condition=OR Boolean(A=Get bIsInFrontEnd(), B=Get isPickaxeEnabled() -> (isPickaxeEnabled:bool)))
   n9 = Lerp(A=1, B=0, Alpha=@n5)
   n10 = "Clamp (Float)"(Value=@n9, Min=0, Max=0.1)
   > Set ConstraintValue1(ConstraintValue1=@n10)
   n11 = "Clamp (Float)"(Value=@n9, Min=0, Max=0.7)
   > Set ConstraintValue2(ConstraintValue2=@n11)
   > Set headUp(headUp=0)
   n12 = Get bIsCrouching()
   n13 = Branch(execute=@n8.else, Condition=@n12)
   > Set ConstraintValue1(ConstraintValue1=@n10)
   > Set ConstraintValue2(ConstraintValue2=@n11)
   n14 = FInterpTo(Current=Get isCrounchingFloat(), Target="To Float (Boolean)"(InBool=@n12), DeltaTime=@n1.DeltaTimeX, InterpSpeed=0.5)
   > Set headUp(headUp=float * float(A=float * float(A=@n14, B=@n6), B=1) {is=title K2Node_PromotableOperator})
   Set ConstraintValue1(execute=@n13.else, ConstraintValue1=@n10)
   > Set ConstraintValue2(ConstraintValue2=@n11)
   > Set headUp(headUp=@n6)
   n15 = CheckForBackbling()
   Set isWearingBackbling(
    execute=@n2.then_2
    isWearingBackbling=AND Boolean!(A=@n15, B=NOT Boolean(A=@n15.bIsBackblingACape), C:bool=NOT Boolean(A=@n15.bIsBackblingLong)) -> (ReturnValue)
   )
   > Set Is Wearing Cape(Is Wearing Cape=OR Boolean(A=@n15.bIsBackblingACape, B=@n15.bIsBackblingLong))
   n16 = Cast To FortPlayerAnimInstance(execute=@n2.then_3, Object=GetAnimInstance(self=GetOwnerSkeletalMeshComponent()))
   > Set isPickaxeEnabled(isPickaxeEnabled:bool=false) -> (Output_Get:bool)
   > Switch on EFortWeaponCoreAnimation!(execute, Selection:EFortWeaponCoreAnimation Enum=Get WeaponCoreAnim(self=@n16)) -> (
    Melee:Exec
    Pistol:Exec
    Shotgun:Exec
    PaperBlueprint:Exec
    Rifle:Exec
    MeleeOneHand:Exec
    MachinePistol:Exec
    RocketLauncher:Exec
    GrenadeLauncher:Exec
    GoingCommando:Exec
    AssaultRifle:Exec
    TacticalShotgun:Exec
    SniperRifle:Exec
    TrapPlacement:Exec
    ShoulderLauncher:Exec
    AbilityDecoTool:Exec
    Crossbow:Exec
    C4:Exec
    RemoteControl:Exec
    DualWield:Exec
    AR_BullPup:Exec
    AR_ForwardGrip:Exec
    MedPackPaddles:Exec
    SMG_P90:Exec
    AR_DrumGun:Exec
    Consumable_Small:Exec
    Consumable_Large:Exec
    Balloon:Exec
    MountedTurret:Exec
    CreativeTool:Exec
    ExplosiveBow:Exec
    AshtonIndigo:Exec
    AshtonChicago:Exec
    MeleeDualWield:Exec
    Unarmed:Exec
   )
   > Set isPickaxeEnabled(isPickaxeEnabled=true)
   n17 = GetRealTimeSeconds()
   n18 = "Sin (Radians)"(A=float * float(A=float * float(A=1, B=1.5), B=@n17) {is=title K2Node_CommutativeAssociativeBinaryOperator})
   n19 = "Sin (Radians)"(A=float * float(A=float * float(A=1, B=1), B=@n17))
   n20 = "Sin (Radians)"(A=float * float(A=float * float(A=1, B=2), B=@n17))
   Set Translate Crown(
    execute=@n2.then_1
    Translate Crown:Vector=MakeVector(X=float * float(A=@n18, B=0.3), Y=float * float(A=@n19, B=0.56), Z=float * float(A=@n20, B=0.22))
   ) -> (Output_Get:Vector)
   n21 = float * float(A=0.6678, B=1.5)
   n22 = float * float(A="Sin (Radians)"(A=float * float(A=@n21, B=@n17)), B=0.85)
   n23 = float * float(A=1, B=1.47)
   n24 = float * float(A="Sin (Radians)"(A=float * float(A=@n23, B=@n17)), B=0.7)
   n25 = float * float(A=1, B=0.953)
   n26 = float * float(A="Sin (Radians)"(A=float * float(A=@n25, B=@n17)), B=0.1)
   n27 = BreakVector(InVec=vector * float(A=MakeVector(X=@n22, Y=@n26, Z=@n24), B=5))
   > Set FloatingCrownRot(FloatingCrownRot:Rotator=MakeRotator(Roll=@n27, Pitch=@n27.Y, Yaw=@n27.Z)) -> (Output_Get:Rotator)
   Event OnCharacterCustomizationLoaded()
   > n28 = Cast To PlayerPawn_Athena(Object=TryGetPawnOwner())
   > Set Head_Mesh(Head_Mesh:Skeletal Mesh Component Object Reference=GetSkeletalMeshForPartType(self=@n28, PartType=Head)) -> (Output_Get:Skeletal Mesh Component Object Reference)
   > n29 = Cast To PlayerPawn_Athena(Object=TryGetPawnOwner())
   > Set Body_Mesh(Body_Mesh=GetSkeletalMeshForPartType(self=@n29))
   "To Float (Boolean)"(InBool=Get bIsTacticalSprinting())
  ]}
]
variables [
 Body_Mesh {type=SkeletalMeshComponent}
 ConstraintValue1 {type=number}
 ConstraintValue2 {type=number}
 emotePlay {type=number}
 FloatingCrownRot {type=Rotator}
 Gravity Override {type=Vector}
 Gravity Override Params {type=GravityOverrideParamsStruct}
 Head_Mesh {type=SkeletalMeshComponent}
 headDownRot {type=number}
 headUp {type=number}
 Is Wearing Cape {type=boolean}
 isCrounchingFloat {type=number}
 isPickaxeEnabled {type=boolean}
 isTargetingFloat {type=number}
 isWearingBackbling {type=boolean}
 playingEmote {type=number}
 skydiving {type=number}
 thinghL Rotation {type=number}
 Translate Crown {type=Vector}
]
